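# Physics asset for the SerpentCoil medium male body's lower vest, set up for secondary motion on three back vest flaps and a left thigh pouch.
# Capsules on pelvis, spine_01 and both thighs form the body core.
# Tiny MSTR anchor capsules attach flaps to spine_01 and pouch to thigh_l with free rotation.
# Child capsules hang off them with tight, zero-twist angular limits, the pouch forming a two-segment chain.
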PhysicsAsset /BRCosmetics/Characters/Player/Male/Medium/Bodies/M_MED_SerpentCoil/Meshes/M_MED_SerpentCoil_LowerVest_Physics
bodies {
 dyn_lowerVestBack_c_01 [{shapeType=Capsule, length=12, radius=2.5, shapeName=dyn_lowerVestBack_c_01_capsule, center=<3.9052 -3.8 0>, rotation=<rot 90 0 0>}]
 dyn_lowerVestBack_c_MSTR [{shapeType=Capsule, length=0.5, radius=0.5, shapeName=dyn_lowerVestBack_c_MSTR_capsule}]
 dyn_lowerVestBack_l_01 [{shapeType=Capsule, length=12, radius=2.5, shapeName=dyn_lowerVestBack_l_01_capsule, center=<3.9052 -3.8 0>, rotation=<rot 90 0 0>}]
 dyn_lowerVestBack_l_MSTR [{shapeType=Capsule, length=0.5, radius=0.5, shapeName=dyn_lowerVestBack_l_MSTR_capsule}]
 dyn_lowerVestBack_r_01 [{shapeType=Capsule, length=12, radius=2.5, shapeName=dyn_lowerVestBack_r_01_capsule, center=<3.9052 -3.8 0>, rotation=<rot 90 0 0>}]
 dyn_lowerVestBack_r_MSTR [{shapeType=Capsule, length=0.5, radius=0.5, shapeName=dyn_lowerVestBack_r_MSTR_capsule}]
 dyn_thighPouch_l_01 [{shapeType=Capsule, length=6, radius=4, shapeName=dyn_thighPouch_l_01_capsule, center=<6.2234 -5 0>, rotation=<rot 90 0 0>}]
 dyn_thighPouch_l_02 [{shapeType=Capsule, length=11.5, radius=3, shapeName=dyn_thighPouch_l_02_capsule, center=<1.5023 0 4.2>, rotation=<rot 90 0 0>}]
 dyn_thighPouch_l_MSTR [{shapeType=Capsule, length=0.5, radius=0.5, shapeName=dyn_thighPouch_l_MSTR_capsule}]
 pelvis [{shapeType=Capsule, length=4.5, radius=14, center=<-1.0821 2.3297 0>}]
 spine_01 [{shapeType=Capsule, length=7.5, radius=12.5, center=<0 3.1683 0>}]
 thigh_l [{shapeType=Capsule, length=33.5, radius=8, center=<-21.1276 0 0>, rotation=<rot 90 0 0>}]
 thigh_r [{shapeType=Capsule, length=25.5, radius=7, center=<21.1276 0 0>, rotation=<rot 90 0 180>}]
}
constraints [
 {bone1Name=spine_01, bone2Name=pelvis, swing1LimitDegrees=45, swing2LimitDegrees=45, twistLimitDegrees=45}
 {bone1Name=thigh_l, bone2Name=pelvis, swing1LimitDegrees=45, swing2LimitDegrees=45, twistLimitDegrees=45}
 {bone1Name=thigh_r, bone2Name=pelvis, swing1LimitDegrees=45, swing2LimitDegrees=45, twistLimitDegrees=45}
 {bone1Name=dyn_lowerVestBack_r_01, bone2Name=dyn_lowerVestBack_r_MSTR, swing1LimitDegrees=20, swing2LimitDegrees=5, twistLimitDegrees=0}
 {
  bone1Name=dyn_lowerVestBack_c_MSTR
  bone2Name=spine_01
  swing1LimitDegrees=45
  swing1Motion=Free
  swing2LimitDegrees=45
  swing2Motion=Free
  twistLimitDegrees=45
  twistMotion=Free
 }
 {bone1Name=dyn_lowerVestBack_c_01, bone2Name=dyn_lowerVestBack_c_MSTR, swing1LimitDegrees=20, swing2LimitDegrees=5, twistLimitDegrees=0}
 {
  bone1Name=dyn_lowerVestBack_l_MSTR
  bone2Name=spine_01
  swing1LimitDegrees=45
  swing1Motion=Free
  swing2LimitDegrees=45
  swing2Motion=Free
  twistLimitDegrees=45
  twistMotion=Free
 }
 {bone1Name=dyn_lowerVestBack_l_01, bone2Name=dyn_lowerVestBack_l_MSTR, swing1LimitDegrees=20, swing2LimitDegrees=5, twistLimitDegrees=0}
 {
  bone1Name=dyn_lowerVestBack_r_MSTR
  bone2Name=spine_01
  swing1LimitDegrees=45
  swing1Motion=Free
  swing2LimitDegrees=45
  swing2Motion=Free
  twistLimitDegrees=45
  twistMotion=Free
 }
 {
  bone1Name=dyn_thighPouch_l_MSTR
  bone2Name=thigh_l
  swing1LimitDegrees=45
  swing1Motion=Free
  swing2LimitDegrees=45
  swing2Motion=Free
  twistLimitDegrees=45
  twistMotion=Free
 }
 {bone1Name=dyn_thighPouch_l_01, bone2Name=dyn_thighPouch_l_MSTR, swing1LimitDegrees=20, swing2LimitDegrees=45, twistLimitDegrees=0}
 {bone1Name=dyn_thighPouch_l_02, bone2Name=dyn_thighPouch_l_01, swing1LimitDegrees=8, swing2LimitDegrees=0, twistLimitDegrees=0}
]
