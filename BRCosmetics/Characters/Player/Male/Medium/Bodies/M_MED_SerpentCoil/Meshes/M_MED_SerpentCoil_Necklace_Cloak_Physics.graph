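# Physics asset for the SerpentCoil medium male body's necklace and back cloak, defining their secondary-motion bodies and joints.
# Capsules on the upper spine and upper arms form the body volumes, joined by free constraints.
# The necklace is a limited-swing chain from spine_05, and the center cloak panel also hangs from spine_05.
# Side and upper cloak panels hang from the upper arms through tiny free-rotating MSTR anchors, with limited swing and zero-degree twist limits.
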PhysicsAsset /BRCosmetics/Characters/Player/Male/Medium/Bodies/M_MED_SerpentCoil/Meshes/M_MED_SerpentCoil_Necklace_Cloak_Physics
bodies {
 dyn_backCloakUp_l_01 [{shapeType=Capsule, length=12, radius=5.5, shapeName=dyn_backCloakUp_l_01_capsule, center=<16 0 8>, rotation=<rot -0.0001 -90 -85.5006>}]
 dyn_backCloakUp_l_MSTR [{shapeType=Capsule, length=0.505, radius=0.505, shapeName=dyn_backCloakUp_l_MSTR_capsule, rotation=<rot 0 -90 0>}]
 dyn_backCloakUp_r_01 [{shapeType=Capsule, length=12, radius=5.5, shapeName=dyn_backCloakUp_r_01_capsule, center=<-19 0 -8>, rotation=<rot -0.0548 -89.9592 93.4805>}]
 dyn_backCloakUp_r_MSTR [{shapeType=Capsule, length=0.505, radius=0.505, shapeName=dyn_backCloakUp_r_MSTR_capsule, rotation=<rot 0 -90 0>}]
 dyn_backCloak_01 [{shapeType=Capsule, length=17.3692, radius=7.5304, shapeName=dyn_backCloak_01_capsule, center=<9 -6 0>, rotation=<rot 0.0002 -90 0.0039>}]
 dyn_backCloak_MSTR [{shapeType=Capsule, length=0.505, radius=0.505, shapeName=dyn_backCloak_MSTR_capsule, rotation=<rot 0 -90 0>}]
 dyn_backCloak_l_01 [{
   shapeType=Capsule
   length=14.5
   radius=7
   shapeName=dyn_backCloak_l_01_capsule
   center=<8.6538 -5.18 0>
   rotation=<rot 7.262 -74.3321 2.0306>
  }]
 dyn_backCloak_l_MSTR [{shapeType=Capsule, length=0.5, radius=0.5, shapeName=dyn_backCloak_l_MSTR_capsule}]
 dyn_backCloak_r_01 [{
   shapeType=Capsule
   length=15.5
   radius=6
   shapeName=dyn_backCloak_r_01_capsule
   center=<8.5878 -5.18 0>
   rotation=<rot -6.9289 -74.3437 -1.9364>
  }]
 dyn_backCloak_r_MSTR [{shapeType=Capsule, length=0.5, radius=0.5, shapeName=dyn_backCloak_r_MSTR_capsule}]
 dyn_necklace_01 [{shapeType=Capsule, length=8, radius=2, shapeName=dyn_necklace_01_capsule, center=<2.4241 1.2303 0>}]
 dyn_necklace_02 [{shapeType=Capsule, length=8, radius=2, shapeName=dyn_necklace_02_capsule, center=<2.4241 1.2303 0>, rotation=<rot 90 0 0>}]
 dyn_necklace_04 [{shapeType=Capsule, length=8, radius=2, shapeName=dyn_necklace_04_capsule, center=<2.4241 1.2303 0>, rotation=<rot 90 0 0>}]
 dyn_necklace_MSTR [{shapeType=Capsule, length=0.5, radius=0.5, shapeName=dyn_necklace_MSTR_capsule}]
 spine_03 [{shapeType=Capsule, length=0, radius=13, shapeName=spine_03_capsule, rotation=<rot 0 15.5389 0>}]
 spine_04 [{shapeType=Capsule, length=6, radius=12.5, shapeName=spine_04_capsule, center=<0.7615 0.9484 0>, rotation=<rot 0 15.5389 0>}]
 spine_05 [{shapeType=Capsule, length=13.5, radius=11.5, shapeName=spine_05_capsule, rotation=<rot 0 15.5389 0>}]
 upperarm_l [{shapeType=Capsule, length=24.4204, radius=7.3353, shapeName=upperarm_l_capsule, center=<8.0598 0 0>, rotation=<rot 90 0 0>}]
 upperarm_r [{shapeType=Capsule, length=24.4204, radius=7.3353, shapeName=upperarm_l_capsule, center=<-8.0598 0 0>, rotation=<rot 90 0 180>}]
}
constraints [
 {bone1Name=dyn_necklace_01, bone2Name=dyn_necklace_MSTR, swing1LimitDegrees=20, swing2LimitDegrees=4, twistLimitDegrees=0}
 {bone1Name=dyn_necklace_02, bone2Name=dyn_necklace_01, swing1LimitDegrees=20, swing2LimitDegrees=4, twistLimitDegrees=0}
 {bone1Name=dyn_necklace_04, bone2Name=dyn_necklace_02, swing1LimitDegrees=30, swing2LimitDegrees=4, twistLimitDegrees=0}
 {bone1Name=dyn_necklace_MSTR, bone2Name=spine_05, swing1LimitDegrees=20, swing2LimitDegrees=8, twistLimitDegrees=45}
 {
  bone1Name=spine_05
  bone2Name=spine_04
  swing1LimitDegrees=45
  swing1Motion=Free
  swing2LimitDegrees=45
  swing2Motion=Free
  twistLimitDegrees=45
  twistMotion=Free
 }
 {
  bone1Name=spine_04
  bone2Name=spine_03
  swing1LimitDegrees=45
  swing1Motion=Free
  swing2LimitDegrees=45
  swing2Motion=Free
  twistLimitDegrees=45
  twistMotion=Free
 }
 {
  bone1Name=dyn_backCloak_MSTR
  bone2Name=spine_05
  swing1LimitDegrees=45
  swing1Motion=Free
  swing2LimitDegrees=45
  swing2Motion=Free
  twistLimitDegrees=45
  twistMotion=Free
 }
 {bone1Name=dyn_backCloak_01, bone2Name=dyn_backCloak_MSTR, swing1LimitDegrees=20, swing2LimitDegrees=0, twistLimitDegrees=0}
 {
  bone1Name=dyn_backCloak_l_MSTR
  bone2Name=upperarm_l
  swing1LimitDegrees=45
  swing1Motion=Free
  swing2LimitDegrees=45
  swing2Motion=Free
  twistLimitDegrees=45
  twistMotion=Free
 }
 {bone1Name=dyn_backCloak_l_01, bone2Name=dyn_backCloak_l_MSTR, swing1LimitDegrees=30, swing2LimitDegrees=10, twistLimitDegrees=0}
 {
  bone1Name=dyn_backCloak_r_MSTR
  bone2Name=upperarm_r
  swing1LimitDegrees=45
  swing1Motion=Free
  swing2LimitDegrees=45
  swing2Motion=Free
  twistLimitDegrees=45
  twistMotion=Free
 }
 {bone1Name=dyn_backCloak_r_01, bone2Name=dyn_backCloak_r_MSTR, swing1LimitDegrees=30, swing2LimitDegrees=10, twistLimitDegrees=0}
 {
  bone1Name=upperarm_l
  bone2Name=spine_05
  swing1LimitDegrees=45
  swing1Motion=Free
  swing2LimitDegrees=45
  swing2Motion=Free
  twistLimitDegrees=45
  twistMotion=Free
 }
 {
  bone1Name=upperarm_r
  bone2Name=spine_05
  swing1LimitDegrees=45
  swing1Motion=Free
  swing2LimitDegrees=45
  swing2Motion=Free
  twistLimitDegrees=45
  twistMotion=Free
 }
 {
  bone1Name=dyn_backCloakUp_l_MSTR
  bone2Name=upperarm_l
  swing1LimitDegrees=45
  swing1Motion=Free
  swing2LimitDegrees=45
  swing2Motion=Free
  twistLimitDegrees=45
  twistMotion=Free
 }
 {bone1Name=dyn_backCloakUp_l_01, bone2Name=dyn_backCloakUp_l_MSTR, swing1LimitDegrees=5, swing2LimitDegrees=10, twistLimitDegrees=0}
 {
  bone1Name=dyn_backCloakUp_r_MSTR
  bone2Name=upperarm_r
  swing1LimitDegrees=45
  swing1Motion=Free
  swing2LimitDegrees=45
  swing2Motion=Free
  twistLimitDegrees=45
  twistMotion=Free
 }
 {bone1Name=dyn_backCloakUp_r_01, bone2Name=dyn_backCloakUp_r_MSTR, swing1LimitDegrees=5, swing2LimitDegrees=10, twistLimitDegrees=0}
]
