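# Control rig for the SerpentCoil character's snake, layering procedural motion over its head, tail and shoulder bones.
# A harmonic wave drives the tail; when targeting, the head partly aims at a hand-driven gun control with added noise.
# Otherwise, time-offset copies of player head rotation, weighted by movement, make the snake head trail, with two bones skipped in frontend.
# Crouch offsets apply outside parachuting, while upper-arm pose readers adjust the snake during it.
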ControlRigBlueprint /BRCosmetics/Characters/Player/Male/Medium/Bodies/M_MED_SerpentCoil/Meshes/M_MED_SerpentCoil_Snake_CtrlRig
graphs [
 Add_1_ContainedGraph {nodes [
   n1 = RigVMFunctionEntryNode()
   Add = /Script/RigVM.RigVMFunction_MathVectorAdd(A=@n1.A, B=@n1.B) {is=struct RigVMUnitNode, template="Add::Execute(in A,in B,out Result)"}
   RigVMFunctionReturnNode(Result=&RigVMFunction_MathVectorAdd(A=@Add.Result, B=@n1.C) {template="Add::Execute(in A,in B,out Result)"}.Result)
  ]}
 Add_1_ContainedGraph {nodes [
   n1 = RigVMFunctionEntryNode()
   Add = &RigVMFunction_MathVectorAdd(A=@n1.A, B=@n1.B) {template="Add::Execute(in A,in B,out Result)"}
   RigVMFunctionReturnNode(Result=&RigVMFunction_MathVectorAdd(A=@Add.Result, B=@n1.C) {template="Add::Execute(in A,in B,out Result)"}.Result)
  ]}
 Add_1_ContainedGraph {nodes [
   n1 = RigVMFunctionEntryNode()
   Add = &RigVMFunction_MathVectorAdd(A=@n1.A, B=@n1.B) {template="Add::Execute(in A,in B,out Result)"}
   RigVMFunctionReturnNode(Result=&RigVMFunction_MathVectorAdd(A=@Add.Result, B=@n1.C) {template="Add::Execute(in A,in B,out Result)"}.Result)
  ]}
 Multiply_1_1_ContainedGraph {nodes [
   n1 = RigVMFunctionEntryNode()
   RigVMFunctionReturnNode(Result=/Script/RigVM.RigVMFunction_MathVectorMul(B=@n1.C, A=&RigVMFunction_MathVectorMul(A=@n1.A, B=@n1.B) {template="Multiply::Execute(in A,in B,out Result)"}.Result) {is=struct RigVMUnitNode, template="Multiply::Execute(in A,in B,out Result)"}.Result)
  ]}
 Multiply_1_1_ContainedGraph {nodes [
   n1 = RigVMFunctionEntryNode()
   RigVMFunctionReturnNode(Result=&RigVMFunction_MathVectorMul(B=@n1.C, A=&RigVMFunction_MathVectorMul(A=@n1.A, B=@n1.B) {template="Multiply::Execute(in A,in B,out Result)"}.Result) {template="Multiply::Execute(in A,in B,out Result)"}.Result)
  ]}
 Multiply_1_1_ContainedGraph {nodes [
   n1 = RigVMFunctionEntryNode()
   RigVMFunctionReturnNode(Result=&RigVMFunction_MathVectorMul(B=@n1.C, A=&RigVMFunction_MathVectorMul(A=@n1.A, B=@n1.B) {template="Multiply::Execute(in A,in B,out Result)"}.Result) {template="Multiply::Execute(in A,in B,out Result)"}.Result)
  ]}
 RigVMFunctionLibrary {links [], nodes []}
 RigVMModel {nodes [
   RigUnit_BeginExecution = &RigUnit_BeginExecution()
   /Script/ControlRig.RigUnit_ChainHarmonicsPerItem(
    ChainRoot=(Type=Bone,Name="dyn_snakeTail_01")
    DrawWorldOffset=(Rotation=(X=0,Y=0,Z=0,W=1),Translation=(X=0,Y=0,Z=0),Scale3D=(X=1,Y=1,Z=1))
    Pendulum=(bEnabled=False,PendulumStiffness=2,PendulumGravity=(X=0,Y=0,Z=0),PendulumBlend=0.75,PendulumDrag=0.98,PendulumMinimum=0,PendulumMaximum=0.1,PendulumEase=Linear,UnwindAxis=(X=0,Y=1,Z=0),UnwindMinimum=0.2,UnwindMaximum=0.05)
    Reach=(bEnabled=False,ReachTarget=(X=0,Y=0,Z=0),ReachAxis=(X=1,Y=0,Z=0),ReachMinimum=0,ReachMaximum=0,ReachEase=Linear)
    Speed=(X=0,Y=0.6,Z=0)
    Wave=(bEnabled=True,WaveFrequency=(X=1,Y=1,Z=1),WaveAmplitude=(X=1,Y=1,Z=0),WaveOffset=(X=0,Y=1,Z=2),WaveNoise=(X=0,Y=1.5,Z=0),WaveMinimum=0,WaveMaximum=1,WaveEase=Linear)
    WaveCurve=(EditorCurveData=(Keys=((),(Time=1,Value=1)),DefaultValue=340282346638528859811704183484516925440,PreInfinityExtrap=RCCE_Constant,PostInfinityExtrap=RCCE_Constant),ExternalCurve=None)
    bDrawDebug=True
    ExecutePin=@RigUnit_BeginExecution.ExecutePin
   ) {is=struct RigVMUnitNode}
   > /Script/ControlRig.RigUnit_PositionConstraintLocalSpaceOffset(
    Child=(Type=Control,Name="snakeAim_MSTR")
    Filter=(bX=True,bY=True,bZ=True)
    Parents=((Item=(Type=Bone,Name="ik_hand_r"),Weight=0.5),(Item=(Type=Bone,Name="ik_hand_l"),Weight=0.5))
    Weight=1
    bMaintainOffset=false
   ) {is=struct RigVMUnitNode}
   > &RigUnit_SetTransform(
    Item=(Type=Control,Name="snakeAim_MSTR")
    Value=(Rotation=(X=0,Y=0,Z=0,W=1),Translation=(X=0,Y=0,Z=0),Scale3D=(X=1,Y=1,Z=1))
    Weight=0
    bInitial=False
    Value.Translation=&RigUnit_GetTransform(Item=(Type=Control,Name="snakeAim_MSTR"), bInitial=False).Transform.Translation
   )
   > n1 = /Script/ControlRig.RigUnit_AimItem(
    DebugSettings=(bEnabled=False,Scale=10,WorldOffset=(Rotation=(X=0,Y=0,Z=0,W=1),Translation=(X=0,Y=0,Z=0),Scale3D=(X=1,Y=1,Z=1)))
    Item=(Type=Control,Name="hand_r_ctrl")
    Primary=(Weight=1,Axis=(X=1,Y=0,Z=0),Target=(X=1,Y=0,Z=0),Kind=Location,Space=(Type=Bone,Name="index_02_r"))
    Secondary=(Weight=1,Axis=(X=0,Y=0,Z=1),Target=(X=1,Y=0,Z=0),Kind=Location,Space=(Type=Bone,Name="None"))
    Weight=1
   ) {is=struct RigVMUnitNode}
   RigVMFunction_ControlFlowBranch = &RigVMFunction_ControlFlowBranch(ExecuteContext=@n1.ExecutePin, Condition=RigVMVariableNode() {variable=isTargeting}.Value)
   n2 = &RigVMFunction_ControlFlowBranch(ExecuteContext=@RigVMFunction_ControlFlowBranch.False, Condition=RigVMVariableNode() {variable=isFrontend}.Value)
   n3 = &RigUnit_GetTransform(Item=(Type=Bone,Name="dyn_snakeHead_04"), Space=LocalSpace, bInitial=False)
   RigUnit_GetRelativeTransformForItem = &RigUnit_GetRelativeTransformForItem(Child=(Type=Bone,Name="head"), Parent=(Type=Bone,Name="spine_05"), bChildInitial=False, bParentInitial=False)
   RigVMFunction_MathQuaternionToEuler = &RigVMFunction_MathQuaternionToEuler(RotationOrder=XYZ, Value=@RigUnit_GetRelativeTransformForItem.RelativeTransform.Rotation)
   RigVMFunction_MathTransformFromSRT = /Script/RigVM.RigVMFunction_MathTransformFromSRT(
    Rotation=(X=0,Y=0,Z=0)
    RotationOrder=XYZ
    Scale=@n3.Transform.Scale3D
    Location=@n3.Transform.Translation
    Rotation.Z=@RigVMFunction_MathQuaternionToEuler.Result.X
    Rotation.X=@RigVMFunction_MathQuaternionToEuler.Result.Y
    Rotation.Y=@RigVMFunction_MathQuaternionToEuler.Result.Z
   ) {is=struct RigVMUnitNode}
   &RigUnit_SetTransform(
    Item=(Type=Bone,Name="dyn_snakeHead_04")
    Space=LocalSpace
    bInitial=False
    ExecutePin=@n2.Completed
    Value=/Script/RigVM.RigVMFunction_TimeOffsetTransform(BufferSize=16, SecondsAgo=0.1, TimeRange=0.8, Value=@RigVMFunction_MathTransformFromSRT.Transform) {is=struct RigVMUnitNode, template="TimeOffset::Execute(in Value,in SecondsAgo,in BufferSize,in TimeRange,out Result)"}.Result
    Weight=RigVMVariableNode() {variable=isPlayerMoving}.Value
   )
   Multiply_1_1_1 = RigVMAggregateNode(
    A=(X=1,Y=1,Z=1)
    B=(X=1,Y=1,Z=1)
    C=(X=0.1,Y=0.1,Z=0.1)
    A.X=@RigVMFunction_MathQuaternionToEuler.Result.X
    A.Y=@RigVMFunction_MathQuaternionToEuler.Result.Y
    A.Z=@RigVMFunction_MathQuaternionToEuler.Result.Z
   )
   n4 = &RigUnit_GetTransform(Item=(Type=Bone,Name="dyn_snakeHead_01"), Space=LocalSpace, bInitial=False)
   n5 = &RigVMFunction_MathQuaternionToEuler(RotationOrder=XYZ, Value=@n4.Transform.Rotation)
   Add_1_1 = RigVMAggregateNode(
    A=(X=0,Y=0,Z=0)
    B=(X=0,Y=0,Z=0)
    C=(X=0,Y=0,Z=0)
    A.Z=@Multiply_1_1_1.Result.X
    A.X=@Multiply_1_1_1.Result.Y
    A.Y=@Multiply_1_1_1.Result.Z
    B.X=@n5.Result.X
    B.Y=@n5.Result.Y
    B.Z=@n5.Result.Z
   )
   > n6 = &RigUnit_SetTransform(
    Item=(Type=Bone,Name="dyn_snakeHead_01")
    Space=LocalSpace
    bInitial=False
    Value=&RigVMFunction_TimeOffsetTransform(
     BufferSize=20
     SecondsAgo=0.2
     TimeRange=0.1
     Value=&RigVMFunction_MathTransformFromSRT(
      Rotation=(X=0,Y=0,Z=0)
      RotationOrder=XYZ
      Rotation.X=@Add_1_1.Result.X
      Rotation.Y=@Add_1_1.Result.Y
      Rotation.Z=@Add_1_1.Result.Z
      Scale=@n4.Transform.Scale3D
      Location=@n4.Transform.Translation
     ).Transform
    ) {template="TimeOffset::Execute(in Value,in SecondsAgo,in BufferSize,in TimeRange,out Result)"}.Result
    Weight=RigVMVariableNode() {variable=isPlayerMoving}.Value
   )
   n7 = &RigVMFunction_ControlFlowBranch(ExecuteContext=@n6.ExecutePin, Condition=RigVMVariableNode() {variable=isParachuteOpen}.Value)
   &RigUnit_ModifyTransforms(
    ItemToModify=((Item=(Type=Bone,Name="dyn_snakeHead_02"),Transform=(Rotation=(X=0,Y=-0.1736,Z=0,W=0.9848),Translation=(X=0,Y=0,Z=0),Scale3D=(X=1,Y=1,Z=1))))
    WeightMinimum=0
    ExecutePin=@n7.False
    Weight=RigVMVariableNode() {variable=isCrouchMoving}.Value
   )
   > &RigUnit_ModifyTransforms(
    ItemToModify=((Item=(Type=Bone,Name="dyn_snakeHead_03"),Transform=(Rotation=(X=-0.0261,Y=-0.2575,Z=0.0974,W=0.961),Translation=(X=0,Y=0,Z=0),Scale3D=(X=1,Y=1,Z=1))))
    WeightMinimum=0
    Weight=RigVMVariableNode() {variable=isCrouchMoving}.Value
   )
   n8 = &RigUnit_SphericalPoseReader(
    ActiveRegionScaleFactors=(PositiveWidth=2.5,NegativeWidth=3,PositiveHeight=3,NegativeHeight=3)
    ActiveRegionSize=0.01
    Debug=(bDrawDebug=true,DebugScale=25,DebugSegments=20,DebugThickness=0.25)
    DriverAxis=(X=1,Y=0,Z=0)
    DriverItem=(Type=Bone,Name="upperarm_l")
    FalloffRegionScaleFactors=(PositiveWidth=1,NegativeWidth=1,PositiveHeight=1,NegativeHeight=1)
    FalloffSize=0.25
    OptionalParentItem=(Type=Bone)
    RotationOffset=(X=0,Y=-50,Z=0)
    ExecutePin=@n7.True
   )
   > &RigUnit_ModifyTransforms(
    ItemToModify=((Item=(Type=Bone,Name="dyn_snakeTail_MSTR"),Transform=(Rotation=(X=-0.0721,Y=0.0079,Z=-0.1085,W=0.9914),Translation=(X=0,Y=0,Z=1),Scale3D=(X=1,Y=1,Z=1))))
    WeightMinimum=0
    Weight=@n8.OutputParam
   )
   > &RigUnit_ModifyTransforms(
    ItemToModify=((Item=(Type=Bone,Name="dyn_snakeShoulder_l_01"),Transform=(Rotation=(X=0,Y=0,Z=0,W=1),Translation=(X=-1,Y=0,Z=-1),Scale3D=(X=1,Y=1,Z=1))))
    WeightMinimum=0
    Weight=@n8.OutputParam
   )
   > n9 = &RigUnit_SphericalPoseReader(
    ActiveRegionScaleFactors=(PositiveWidth=2.5,NegativeWidth=3,PositiveHeight=3,NegativeHeight=3)
    ActiveRegionSize=0.01
    Debug=(bDrawDebug=true,DebugScale=25,DebugSegments=20,DebugThickness=0.25)
    DriverAxis=(X=-1,Y=0,Z=0)
    DriverItem=(Type=Bone,Name="upperarm_r")
    FalloffRegionScaleFactors=(PositiveWidth=1,NegativeWidth=1,PositiveHeight=1,NegativeHeight=1)
    FalloffSize=0.4
    FlipHeightScaling=true
    OptionalParentItem=(Type=Bone)
    RotationOffset=(X=180,Y=50,Z=180)
   )
   > &RigUnit_ModifyTransforms(
    ItemToModify=((Item=(Type=Bone,Name="dyn_snakeHead_MSTR"),Transform=(Rotation=(X=0,Y=-0.0872,Z=0,W=0.9962),Translation=(X=0,Y=0,Z=0),Scale3D=(X=1,Y=1,Z=1))))
    WeightMinimum=0
    Weight=@n9.OutputParam
   )
   > &RigUnit_ModifyTransforms(
    ItemToModify=((Item=(Type=Bone,Name="dyn_snakeShoulder_r_01"),Transform=(Rotation=(X=0,Y=0,Z=0,W=1),Translation=(X=-1,Y=0,Z=1),Scale3D=(X=1,Y=1,Z=1))))
    WeightMinimum=0
    Weight=@n9.OutputParam
   )
   Multiply_1 = RigVMAggregateNode(
    A=(X=1,Y=1,Z=1)
    B=(X=1,Y=1,Z=1)
    C=(X=0.1,Y=0.3,Z=0.3)
    A.X=@RigVMFunction_MathQuaternionToEuler.Result.X
    A.Y=@RigVMFunction_MathQuaternionToEuler.Result.Y
    A.Z=@RigVMFunction_MathQuaternionToEuler.Result.Z
   )
   n10 = &RigUnit_GetTransform(Item=(Type=Bone,Name="dyn_snakeHead_02"), Space=LocalSpace, bInitial=False)
   n11 = &RigVMFunction_MathQuaternionToEuler(RotationOrder=XYZ, Value=@n10.Transform.Rotation)
   Add = RigVMAggregateNode(
    A=(X=0,Y=0,Z=0)
    B=(X=0,Y=0,Z=0)
    C=(X=0,Y=0,Z=0)
    A.Z=@Multiply_1.Result.X
    A.X=@Multiply_1.Result.Y
    A.Y=@Multiply_1.Result.Z
    B.X=@n11.Result.X
    B.Y=@n11.Result.Y
    B.Z=@n11.Result.Z
   )
   &RigUnit_SetTransform(
    Item=(Type=Bone,Name="dyn_snakeHead_02")
    Space=LocalSpace
    bInitial=False
    Value=&RigVMFunction_TimeOffsetTransform(
     BufferSize=20
     SecondsAgo=0.1
     TimeRange=1
     Value=&RigVMFunction_MathTransformFromSRT(
      Rotation=(X=0,Y=0,Z=0)
      RotationOrder=XYZ
      Rotation.X=@Add.Result.X
      Rotation.Y=@Add.Result.Y
      Rotation.Z=@Add.Result.Z
      Scale=@n10.Transform.Scale3D
      Location=@n10.Transform.Translation
     ).Transform
    ) {template="TimeOffset::Execute(in Value,in SecondsAgo,in BufferSize,in TimeRange,out Result)"}.Result
    Weight=RigVMVariableNode() {variable=isPlayerMoving}.Value
    ExecutePin=@n2.False
   )
   Multiply_1_1 = RigVMAggregateNode(
    A=(X=1,Y=1,Z=1)
    B=(X=1,Y=1,Z=1)
    C=(X=0.2,Y=0.3,Z=0.3)
    A.X=@RigVMFunction_MathQuaternionToEuler.Result.X
    A.Y=@RigVMFunction_MathQuaternionToEuler.Result.Y
    A.Z=@RigVMFunction_MathQuaternionToEuler.Result.Z
   )
   n12 = &RigUnit_GetTransform(Item=(Type=Bone,Name="dyn_snakeHead_03"), Space=LocalSpace, bInitial=False)
   n13 = &RigVMFunction_MathQuaternionToEuler(RotationOrder=XYZ, Value=@n12.Transform.Rotation)
   Add_1 = RigVMAggregateNode(
    A=(X=0,Y=0,Z=0)
    B=(X=0,Y=0,Z=0)
    C=(X=0,Y=0,Z=0)
    A.Z=@Multiply_1_1.Result.X
    A.X=@Multiply_1_1.Result.Y
    A.Y=@Multiply_1_1.Result.Z
    B.X=@n13.Result.X
    B.Y=@n13.Result.Y
    B.Z=@n13.Result.Z
   )
   > &RigUnit_SetTransform(
    Item=(Type=Bone,Name="dyn_snakeHead_03")
    Space=LocalSpace
    bInitial=False
    Value=&RigVMFunction_TimeOffsetTransform(
     BufferSize=20
     SecondsAgo=0.2
     TimeRange=1
     Value=&RigVMFunction_MathTransformFromSRT(
      Rotation=(X=0,Y=0,Z=0)
      RotationOrder=XYZ
      Rotation.X=@Add_1.Result.X
      Rotation.Y=@Add_1.Result.Y
      Rotation.Z=@Add_1.Result.Z
      Scale=@n12.Transform.Scale3D
      Location=@n12.Transform.Translation
     ).Transform
    ) {template="TimeOffset::Execute(in Value,in SecondsAgo,in BufferSize,in TimeRange,out Result)"}.Result
    Weight=RigVMVariableNode() {variable=isPlayerMoving}.Value
   )
   &RigUnit_AimItem(
    DebugSettings=(bEnabled=False,Scale=10,WorldOffset=(Rotation=(X=0,Y=0,Z=0,W=1),Translation=(X=0,Y=0,Z=0),Scale3D=(X=1,Y=1,Z=1)))
    Item=(Type=Bone,Name="dyn_snakeHead_02")
    Primary=(Weight=1,Axis=(X=1,Y=0,Z=0),Target=(X=0,Y=0,Z=1),Kind=Location,Space=(Type=Control,Name="gunAim_ctrl"))
    Secondary=(Weight=1,Axis=(X=0,Y=0,Z=1),Target=(X=1,Y=0,Z=0),Kind=Location,Space=(Type=Bone,Name="dyn_snakeUpVector"))
    Weight=0.3
    ExecutePin=@RigVMFunction_ControlFlowBranch.True
   )
   Noise = /Script/RigVM.RigVMFunction_NoiseVector2(Frequency=(X=0.4,Y=0.5,Z=0.3), Maximum=10, Minimum=0, Speed=(X=0.3,Y=0.6,Z=0.3), Value=(X=6,Y=7,Z=6)) {is=struct RigVMUnitNode, template="Noise::Execute(in Value,in Speed,in Frequency,in Minimum,in Maximum,out Result)"}
   > &RigUnit_ModifyTransforms(
    ItemToModify=((Item=(Type=Bone,Name="dyn_snakeHead_03"),Transform=(Rotation=(X=0,Y=0,Z=0,W=1),Translation=(X=0,Y=0,Z=0),Scale3D=(X=1,Y=1,Z=1))))
    Weight=1
    WeightMinimum=0
    ItemToModify.0.Transform.Rotation=&RigVMFunction_MathQuaternionFromEuler(Euler=@Noise.Result).Result
   )
   > &RigUnit_ModifyTransforms(
    ItemToModify=((Item=(Type=Bone,Name="dyn_snakeHead_02"),Transform=(Rotation=(X=0,Y=0,Z=0,W=1),Translation=(X=0,Y=0,Z=0),Scale3D=(X=1,Y=1,Z=1))))
    Weight=1
    WeightMinimum=0
    ItemToModify.0.Transform.Rotation=&RigVMFunction_MathQuaternionFromEuler(Euler=&RigVMFunction_NoiseVector2(Frequency=(X=0.4,Y=0.8,Z=0.3), Maximum=5, Minimum=0, Speed=(X=0.2,Y=0.8,Z=1), Value=(X=2,Y=5,Z=5)) {template="Noise::Execute(in Value,in Speed,in Frequency,in Minimum,in Maximum,out Result)"}.Result).Result
   )
   > &RigUnit_AimItem(
    DebugSettings=(bEnabled=False,Scale=10,WorldOffset=(Rotation=(X=0,Y=0,Z=0,W=1),Translation=(X=0,Y=0,Z=0),Scale3D=(X=1,Y=1,Z=1)))
    Item=(Type=Bone,Name="dyn_snakeHead_04")
    Primary=(Weight=1,Axis=(X=1,Y=0,Z=0),Target=(X=0,Y=0,Z=1),Kind=Location,Space=(Type=Control,Name="gunAim_ctrl"))
    Secondary=(Weight=1,Axis=(X=0,Y=0,Z=1),Target=(X=1,Y=0,Z=0),Kind=Location,Space=(Type=Bone,Name="dyn_snakeUpVector"))
    Weight=0
   )
  ]}
]
hierarchy [
 root {loc=<0 0 0>, rot=<0 0 0 1>}
 pelvis {parent=root, loc=<0 -0.7071 93.1191>, rot=<0.0127 0.707 -0.0127 -0.707>}
 spine_01 {parent=pelvis, loc=<6.7494 -0.6388 0>, rot=<0 0 0.034 -0.9994>}
 spine_02 {parent=spine_01, loc=<7.2518 -0.035 0>, rot=<0 0 0.0238 -0.9997>}
 spine_03 {parent=spine_02, loc=<6.0104 0 0>, rot=<0 0 0.0149 -0.9999>}
 spine_04 {parent=spine_03, loc=<8.1024 0 0>, rot=<0 0 -0.0132 -0.9999>}
 spine_05 {parent=spine_04, loc=<10.0146 0 0>, rot=<0 0 -0.0015 -1>}
 attach_cape {parent=spine_05, loc=<9.0117 -11.7049 -0.0007>, rot=<0.0616 -0.7044 -0.0616 -0.7044>}
 neck_01 {parent=spine_05, loc=<11.3138 0.1529 0>, rot=<0 0 -0.2003 -0.9797>}
 neck_02 {parent=neck_01, loc=<4.5627 0 0>, rot=<0 0 0.0504 -0.9987>}
 head {parent=neck_02, loc=<5.3914 0 0>, rot=<0 0 0.115 -0.9934>}
 dyn_snakeAimHead {parent=head, loc=<-0.0511 54.6915 15.3346>, rot=<-0.4809 -0.5184 -0.4852 -0.5144>}
 dyn_snakeAimHead_ctrl {controlType=EULER_TRANSFORM, parent=dyn_snakeAimHead, type=Control, loc=<0 0 0>, rot=<0 0 0 1>}
 attach_backpack {parent=spine_05, loc=<-0.1228 -13.4628 -0.0017>, rot=<-0.0093 -0.707 0.0093 -0.707>}
 clavicle_l {parent=spine_05, loc=<4.3575 -0.1466 -2.3223>, rot=<0.0465 -0.7289 -0.009 -0.6829>}
 upperarm_l {parent=clavicle_l, loc=<15.5111 0 0>, rot=<0.0226 -0.3234 -0.0088 -0.9459>}
 lowerarm_l {parent=upperarm_l, loc=<26.6514 0 0>, rot=<0.0391 -0.0099 -0.2453 -0.9686>}
 hand_l {parent=lowerarm_l, loc=<22.3491 0 0>, rot=<-0.7081 -0.0507 0.0975 -0.6975>}
 thumb_01_l {parent=hand_l, loc=<1.3118 -0.3191 -2.459>, rot=<0.6546 -0.4042 0.0664 -0.6354>}
 thumb_02_l {parent=thumb_01_l, loc=<5.9336 -0.0448 0.0401>, rot=<-0.0037 0.0213 0.298 -0.9543>}
 thumb_03_l {parent=thumb_02_l, loc=<2.2223 0.3682 -0.0907>, rot=<0.0118 -0.013 0.0066 -0.9998>}
 index_metacarpal_l {parent=hand_l, loc=<1.6535 0.0183 -2.96>, rot=<0.0158 -0.0124 -0.0173 -0.9996>}
 index_01_l {parent=index_metacarpal_l, loc=<8.3008 0.0357 -0.0524>, rot=<-0.0044 0.0485 0.0924 -0.9945>}
 index_02_l {parent=index_01_l, loc=<3.3587 -0.3004 0.0452>, rot=<-0.002 0.0027 0.3225 -0.9466>}
 index_03_l {parent=index_02_l, loc=<2.5578 -0.123 -0.0252>, rot=<-0.002 0.0067 0.2558 -0.9667>}
 middle_metacarpal_l {parent=hand_l, loc=<2.1856 -0.1725 -0.5984>, rot=<0.006 0.0168 -0.0329 -0.9993>}
 middle_01_l {parent=middle_metacarpal_l, loc=<7.3275 -0.2447 0.0129>, rot=<-0.0208 0.0042 0.0858 -0.9961>}
 middle_02_l {parent=middle_01_l, loc=<4.0123 0.1401 0.1175>, rot=<-0.0218 0.0258 0.3512 -0.9357>}
 middle_03_l {parent=middle_02_l, loc=<2.6077 0.0534 0.1373>, rot=<0.0082 0.0188 0.1708 -0.9851>}
 ring_metacarpal_l {parent=hand_l, loc=<2.2074 -0.3625 1.3638>, rot=<-0.0284 0.0722 -0.0458 -0.9959>}
 ring_01_l {parent=ring_metacarpal_l, loc=<7.0295 -0.6197 -0.1446>, rot=<-0.0489 -0.0252 0.1892 -0.9804>}
 ring_02_l {parent=ring_01_l, loc=<3.6174 0.4647 -0.1027>, rot=<0 0 0.2806 -0.9598>}
 ring_03_l {parent=ring_02_l, loc=<2.4152 0.114 0.0199>, rot=<0 0 0.1189 -0.9929>}
 pinky_metacarpal_l {parent=hand_l, loc=<1.7616 -0.9753 3.3814>, rot=<-0.0397 0.0814 -0.069 -0.9935>}
 pinky_01_l {parent=pinky_metacarpal_l, loc=<7.1709 -0.4831 -0.1094>, rot=<-0.0654 -0.0035 0.2242 -0.9723>}
 pinky_02_l {parent=pinky_01_l, loc=<2.8274 0.0893 -0.1043>, rot=<0.0037 0.0011 0.2854 -0.9584>}
 pinky_03_l {parent=pinky_02_l, loc=<2.2046 0.0307 -0.0078>, rot=<-0.0002 0.0098 0.1445 -0.9895>}
 deform_wrist_dn_base_l {parent=hand_l, loc=<0 -2.2342 0>, rot=<0 0 0 1>}
 deform_wrist_dn_l {parent=deform_wrist_dn_base_l, loc=<0 0 0>, rot=<0 0 0 1>}
 deform_wrist_up_base_l {parent=hand_l, loc=<0 2.2682 0>, rot=<0 0 0 1>}
 deform_wrist_up_l {parent=deform_wrist_up_base_l, loc=<0 0 0>, rot=<0 0 0 1>}
 lowerarm_twist_02_l {parent=lowerarm_l, loc=<9.6199 0 0>, rot=<0 0 0 -1>}
 lowerarm_twist_01_l {parent=lowerarm_l, loc=<17.8549 0 0>, rot=<0 0 0 -1>}
 deform_elbow_in_base_l {parent=lowerarm_l, loc=<0.2112 5.259 0.8274>, rot=<0 0 0 -1>}
 deform_elbow_in_l {parent=deform_elbow_in_base_l, loc=<0 0 0>, rot=<0 0 0 1>}
 deform_elbow_out_base_l {parent=lowerarm_l, loc=<0 -4 0>, rot=<0 0 0 -1>}
 deform_elbow_out_l {parent=deform_elbow_out_base_l, loc=<0 0 0>, rot=<0 0 0 1>}
 upperarm_twist_01_l {parent=upperarm_l, loc=<0.0041 0 0>, rot=<0 0 0 1>}
 deform_upperarm_bk_base_l {parent=upperarm_twist_01_l, loc=<1.0813 -7.2193 -1.9453>, rot=<0 0 0 1>}
 deform_upperarm_bk_l {parent=deform_upperarm_bk_base_l, loc=<0 0 0>, rot=<0 0 0 1>}
 dyn_backCloak_l_MSTR {parent=deform_upperarm_bk_l, loc=<0 0 0>, rot=<-0.0101 -0.1576 0.0101 -0.9874>}
 dyn_backCloak_l_01 {parent=dyn_backCloak_l_MSTR, loc=<0 0 0>, rot=<0.0084 -0.1487 0.056 -0.9873>}
 dyn_backCloak_l_02 {parent=dyn_backCloak_l_01, loc=<10.5172 0 0>, rot=<0 0 0 1>}
 deform_upperarm_fr_base_l {parent=upperarm_twist_01_l, loc=<-0.3691 7.0506 -2.765>, rot=<0 0 0 1>}
 deform_upperarm_fr_l {parent=deform_upperarm_fr_base_l, loc=<0 0 0>, rot=<0 0 0 1>}
 deform_upperarm_up_base_l {parent=upperarm_twist_01_l, loc=<-3.2216 -1.1391 3.3311>, rot=<0 0 0 1>}
 deform_upperarm_up_l {parent=deform_upperarm_up_base_l, loc=<0 0 0>, rot=<0 0 0 1>}
 dyn_backCloakUp_l_MSTR {parent=upperarm_twist_01_l, loc=<-3.2216 -1.1391 3.3311>, rot=<0 0.0436 0 -0.999>}
 dyn_backCloakUp_l_01 {parent=dyn_backCloakUp_l_MSTR, loc=<0 0 0>, rot=<0 0 0 1>}
 dyn_backCloakUp_l_02 {parent=dyn_backCloakUp_l_01, loc=<19 0 0>, rot=<0 0 0 1>}
 dyn_sleeve_l_01 {parent=upperarm_twist_01_l, loc=<5 0 0>, rot=<0 0 0 1>}
 upperarm_twist_02_l {parent=upperarm_l, loc=<13.7941 0 0>, rot=<0 0 0 1>}
 clavicle_r {parent=spine_05, loc=<4.3575 -0.1466 2.3223>, rot=<0.7289 0.0465 -0.6829 0.009>}
 upperarm_r {parent=clavicle_r, loc=<-15.5111 0 0>, rot=<0.0226 -0.3234 -0.0088 -0.9459>}
 lowerarm_r {parent=upperarm_r, loc=<-26.6514 0 0>, rot=<0.0391 -0.0099 -0.2453 -0.9686>}
 hand_r {parent=lowerarm_r, loc=<-22.3491 0 0>, rot=<-0.7081 -0.0507 0.0975 -0.6975>}
 thumb_01_r {parent=hand_r, loc=<-1.3118 0.3191 2.459>, rot=<0.6546 -0.4042 0.0664 -0.6354>}
 thumb_02_r {parent=thumb_01_r, loc=<-5.9336 0.0448 -0.0401>, rot=<-0.0037 0.0213 0.298 -0.9543>}
 thumb_03_r {parent=thumb_02_r, loc=<-2.2223 -0.3682 0.0907>, rot=<0.0118 -0.013 0.0066 -0.9998>}
 index_metacarpal_r {parent=hand_r, loc=<-1.6535 -0.0183 2.96>, rot=<0.0158 -0.0124 -0.0173 -0.9996>}
 index_01_r {parent=index_metacarpal_r, loc=<-8.3008 -0.0357 0.0524>, rot=<-0.0044 0.0485 0.0924 -0.9945>}
 index_02_r {parent=index_01_r, loc=<-3.3587 0.3004 -0.0452>, rot=<-0.002 0.0027 0.3225 -0.9466>}
 index_03_r {parent=index_02_r, loc=<-2.5578 0.123 0.0252>, rot=<-0.002 0.0067 0.2558 -0.9667>}
 middle_metacarpal_r {parent=hand_r, loc=<-2.1856 0.1725 0.5984>, rot=<0.006 0.0168 -0.0329 -0.9993>}
 middle_01_r {parent=middle_metacarpal_r, loc=<-7.3275 0.2447 -0.0129>, rot=<-0.0208 0.0042 0.0858 -0.9961>}
 middle_02_r {parent=middle_01_r, loc=<-4.0123 -0.1401 -0.1175>, rot=<-0.0218 0.0258 0.3512 -0.9357>}
 middle_03_r {parent=middle_02_r, loc=<-2.6077 -0.0534 -0.1373>, rot=<0.0082 0.0188 0.1708 -0.9851>}
 ring_metacarpal_r {parent=hand_r, loc=<-2.2074 0.3625 -1.3638>, rot=<-0.0284 0.0722 -0.0458 -0.9959>}
 ring_01_r {parent=ring_metacarpal_r, loc=<-7.0295 0.6197 0.1446>, rot=<-0.0489 -0.0252 0.1892 -0.9804>}
 ring_02_r {parent=ring_01_r, loc=<-3.6174 -0.4647 0.1027>, rot=<0 0 0.2806 -0.9598>}
 ring_03_r {parent=ring_02_r, loc=<-2.4152 -0.114 -0.0199>, rot=<0 0 0.1189 -0.9929>}
 pinky_metacarpal_r {parent=hand_r, loc=<-1.7616 0.9753 -3.3814>, rot=<-0.0397 0.0814 -0.069 -0.9935>}
 pinky_01_r {parent=pinky_metacarpal_r, loc=<-7.1709 0.4831 0.1094>, rot=<-0.0654 -0.0035 0.2242 -0.9723>}
 pinky_02_r {parent=pinky_01_r, loc=<-2.8274 -0.0893 0.1043>, rot=<0.0037 0.0011 0.2854 -0.9584>}
 pinky_03_r {parent=pinky_02_r, loc=<-2.2046 -0.0307 0.0078>, rot=<-0.0002 0.0098 0.1445 -0.9895>}
 deform_wrist_dn_base_r {parent=hand_r, loc=<-0.0004 2.2345 -0.0001>, rot=<0 0 0 1>}
 deform_wrist_dn_r {parent=deform_wrist_dn_base_r, loc=<0 0 0>, rot=<0 0 0 1>}
 deform_wrist_up_base_r {parent=hand_r, loc=<0.0003 -2.2684 0.0001>, rot=<0 0 0 1>}
 deform_wrist_up_r {parent=deform_wrist_up_base_r, loc=<0 0 0>, rot=<0 0 0 1>}
 hand_r_ctrl {controlType=EULER_TRANSFORM, parent=hand_r, type=Control, loc=<0 0 0>, rot=<0 0 0 1>}
 gunAim_ctrl {controlType=EULER_TRANSFORM, parent=hand_r_ctrl, type=Control, loc=<0 0 0>, rot=<0 0 0 1>}
 lowerarm_twist_02_r {parent=lowerarm_r, loc=<-9.6199 0 0>, rot=<0 0 0 -1>}
 lowerarm_twist_01_r {parent=lowerarm_r, loc=<-17.8549 0 0>, rot=<0 0 0 -1>}
 deform_elbow_in_base_r {parent=lowerarm_r, loc=<-0.2111 -5.2591 -0.8275>, rot=<0 0 0 -1>}
 deform_elbow_in_r {parent=deform_elbow_in_base_r, loc=<-0.0001 0 0.0002>, rot=<0 0 0 1>}
 deform_elbow_out_base_r {parent=lowerarm_r, loc=<0 4 -0.0001>, rot=<0 0 0 -1>}
 deform_elbow_out_r {parent=deform_elbow_out_base_r, loc=<0 0 0>, rot=<0 0 0 1>}
 upperarm_twist_01_r {parent=upperarm_r, loc=<-0.0041 0 0>, rot=<0 0 0 1>}
 deform_upperarm_bk_base_r {parent=upperarm_twist_01_r, loc=<-1.0814 7.2192 1.9456>, rot=<0 0 0 1>}
 deform_upperarm_bk_r {parent=deform_upperarm_bk_base_r, loc=<0 0 0>, rot=<0 0 0 1>}
 dyn_backCloak_r_MSTR {parent=deform_upperarm_bk_r, loc=<0 0 0>, rot=<0.1403 0.0524 0.9877 0.0457>}
 dyn_backCloak_r_01 {parent=dyn_backCloak_r_MSTR, loc=<0 0 0>, rot=<0.0392 0.1688 0.0031 -0.9849>}
 dyn_backCloak_r_02 {parent=dyn_backCloak_r_01, loc=<10.3451 0 0>, rot=<0 0 0 1>}
 deform_upperarm_fr_base_r {parent=upperarm_twist_01_r, loc=<0.3689 -7.0506 2.7652>, rot=<0 0 0 1>}
 deform_upperarm_fr_r {parent=deform_upperarm_fr_base_r, loc=<0 0 0>, rot=<0 0 0 1>}
 deform_upperarm_up_base_r {parent=upperarm_twist_01_r, loc=<3.2214 1.1391 -3.3308>, rot=<0 0 0 1>}
 deform_upperarm_up_r {parent=deform_upperarm_up_base_r, loc=<0 0 0>, rot=<0 0 0 -1>}
 dyn_backCloakUp_r_MSTR {parent=upperarm_twist_01_r, loc=<3.2214 1.1391 -3.3308>, rot=<0 0.0436 0 -0.999>}
 dyn_backCloakUp_r_01 {parent=dyn_backCloakUp_r_MSTR, loc=<0 0 0>, rot=<0 0 0 1>}
 dyn_backCloakUp_r_02 {parent=dyn_backCloakUp_r_01, loc=<-19 0 0>, rot=<0 0 0 1>}
 dyn_sleeve_r_01 {parent=upperarm_twist_01_r, loc=<-5 0 0>, rot=<0 0 0 1>}
 upperarm_twist_02_r {parent=upperarm_r, loc=<-13.7941 0 0>, rot=<0 0 0 1>}
 dyn_necklace_MSTR {parent=spine_05, loc=<3.1035 11.3275 -0.0127>, rot=<-0.1554 -0.9879 0 0>}
 dyn_necklace_01 {parent=dyn_necklace_MSTR, loc=<0 0 0>, rot=<0 0 0 1>}
 dyn_necklace_02 {parent=dyn_necklace_01, loc=<5.2115 0 0>, rot=<0 0 0.0865 -0.9963>}
 dyn_necklace_04 {parent=dyn_necklace_02, loc=<5.8771 0 0>, rot=<0 0 0.1093 -0.994>}
 dyn_snakeChest_01 {parent=spine_05, loc=<4.7964 12.478 0.3716>, rot=<-0.2905 -0.9569 0 0>}
 dyn_backCloak_MSTR {parent=spine_05, loc=<-0.084 -11.9834 0>, rot=<0.0455 -0.999 0 0>}
 dyn_backCloak_01 {parent=dyn_backCloak_MSTR, loc=<0 0 0>, rot=<0 0 0 1>}
 dyn_backCloak_02 {parent=dyn_backCloak_01, loc=<13.2132 0 0>, rot=<0 0 0 1>}
 dyn_snake_MSTR {parent=spine_05, loc=<-10.0146 0.0299 0>, rot=<0 0 0 1>}
 dyn_snakeHead_MSTR {parent=dyn_snake_MSTR, loc=<24.7074 -8.9137 -0.1876>, rot=<0.8118 0.094 0.5749 0.04>}
 dyn_snakeHead_01 {parent=dyn_snakeHead_MSTR, loc=<0 0 0>, rot=<0 0 0 1>}
 dyn_snakeHead_02 {parent=dyn_snakeHead_01, loc=<17.0384 0 0>, rot=<0.1067 -0.0335 0.4396 -0.8912>}
 dyn_snakeHead_03 {parent=dyn_snakeHead_02, loc=<7.4859 0 0>, rot=<0.0589 -0.0979 0.2217 -0.9684>}
 dyn_snakeHead_04 {parent=dyn_snakeHead_03, loc=<6.4987 0 0>, rot=<0 0 0.0334 -0.9994>}
 dyn_snakeHat_01 {parent=dyn_snakeHead_04, loc=<3.2472 0.4114 1.1443>, rot=<-0.5 -0.5 0.5 -0.5>}
 dyn_snakeString_01 {parent=dyn_snakeHead_04, loc=<4.7151 -1.7693 -1.651>, rot=<-0.5 -0.5 0.5 -0.5>}
 dyn_snakeChain_01 {parent=dyn_snakeHead_04, loc=<5.2769 1.4965 -2.5493>, rot=<-0.5 -0.5 0.5 -0.5>}
 dyn_snakeTail_MSTR {parent=dyn_snake_MSTR, loc=<27.1483 -5.8233 -0.6015>, rot=<0.1039 -0.6496 -0.0036 -0.7531>}
 dyn_snakeTail_01 {parent=dyn_snakeTail_MSTR, loc=<0 0 0>, rot=<0 0 0 1>}
 dyn_snakeTail_02 {parent=dyn_snakeTail_01, loc=<5.8501 0 0>, rot=<0.0771 -0.4515 0.0759 -0.8857>}
 dyn_snakeTail_03 {parent=dyn_snakeTail_02, loc=<9.8541 0 0>, rot=<0.1143 -0.3118 -0.1311 -0.9341>}
 dyn_snakeTail_04 {parent=dyn_snakeTail_03, loc=<11.7315 0 0>, rot=<0.0455 0.8392 -0.0955 -0.5335>}
 dyn_snakeTail_05 {parent=dyn_snakeTail_04, loc=<3.3336 0.562 -0.2191>, rot=<0 -0.5333 0 -0.8459>}
 dyn_snakeShoulder_r_01 {parent=dyn_snake_MSTR, loc=<24.4824 1.525 10.8951>, rot=<0.188 -0.9822 0 0>}
 dyn_snakeShoulder_l_01 {parent=dyn_snake_MSTR, loc=<25.4935 0.3803 -9.5797>, rot=<0.1475 -0.9891 0 0>}
 deform_pec_aim_base_l {parent=spine_04, loc=<7.8732 12.6388 -2.4123>, rot=<0.0102 -0.4602 0.3979 -0.7936>}
 deform_pec_aim_l {parent=deform_pec_aim_base_l, loc=<0 0 0>, rot=<0 0 0 1>}
 deform_pec_base_l {parent=deform_pec_aim_l, loc=<10.766 0 0>, rot=<0 0 0 1>}
 deform_pec_l {parent=deform_pec_base_l, loc=<0.055 0 0>, rot=<0 0 0 1>}
 deform_pec_aim_base_r {parent=spine_04, loc=<7.8727 12.6388 2.4123>, rot=<-0.4602 -0.0102 0.7936 0.3979>}
 deform_pec_aim_r {parent=deform_pec_aim_base_r, loc=<0 0 0>, rot=<0 0 0 1>}
 deform_pec_base_r {parent=deform_pec_aim_r, loc=<-10.7662 -0.0006 -0.0007>, rot=<0 0 0 1>}
 deform_pec_r {parent=deform_pec_base_r, loc=<0 0 0>, rot=<0 0 0 1>}
 dyn_snakeUpVector {parent=spine_04, loc=<88.5035 7.3881 20.3168>, rot=<0 0 -0.0015 -1>}
 dyn_zipperVest_MSTR {parent=spine_01, loc=<2.1184 12.184 -1.4653>, rot=<-0.0179 -0.9899 0.1108 -0.0861>}
 dyn_zipperVest_01 {parent=dyn_zipperVest_MSTR, loc=<0 0 0>, rot=<0 0 0 1>}
 dyn_lowerVestBack_c_MSTR {parent=spine_01, loc=<1.1603 -8.3563 -0.0017>, rot=<0.1332 -0.9911 0 0>}
 dyn_lowerVestBack_c_01 {parent=dyn_lowerVestBack_c_MSTR, loc=<0 0 0>, rot=<0 0 0 1>}
 dyn_lowerVestBack_r_MSTR {parent=spine_01, loc=<1.3628 -6.7312 9.0392>, rot=<0.119 -0.9593 0.2518 0.0472>}
 dyn_lowerVestBack_r_01 {parent=dyn_lowerVestBack_r_MSTR, loc=<0 0 0>, rot=<0 0 0 1>}
 dyn_lowerVestBack_l_MSTR {parent=spine_01, loc=<1.1839 -6.7253 -8.9711>, rot=<0.1118 -0.9437 -0.3043 -0.0655>}
 dyn_lowerVestBack_l_01 {parent=dyn_lowerVestBack_l_MSTR, loc=<0 0 0>, rot=<0 0 0 1>}
 thigh_l {parent=pelvis, loc=<-2.5693 -0.0109 -7.7501>, rot=<0.0012 0.0572 0.0155 -0.9982>}
 calf_l {parent=thigh_l, loc=<-39.0484 0 0>, rot=<0.0006 -0.0161 -0.0391 -0.9991>}
 foot_l {parent=calf_l, loc=<-42.561 0.1077 0.4292>, rot=<0.0418 -0.0455 0.0343 -0.9975>}
 ball_l {parent=foot_l, loc=<-6.1599 13.8264 0.6806>, rot=<0.0001 -0.0001 -0.7063 -0.7079>}
 calf_twist_02_l {parent=calf_l, loc=<-18.8064 -0.0967 0.0397>, rot=<0.0237 0.0095 -0.0035 -0.9997>}
 dyn_pantLeg_l_01 {parent=calf_twist_02_l, loc=<0 0 0>, rot=<0 0 0 1>}
 calf_twist_01_l {parent=calf_l, loc=<-35.813 -0.0967 0.0397>, rot=<0.0237 0.0095 -0.0035 -0.9997>}
 deform_knee_in_base_l {parent=calf_l, loc=<1.298 -5.2957 -0.0496>, rot=<0 0 0 -1>}
 deform_knee_in_l {parent=deform_knee_in_base_l, loc=<0 0 0>, rot=<0 0 0 1>}
 deform_knee_out_base_l {parent=calf_l, loc=<0 4 0>, rot=<0 0 0 -1>}
 deform_knee_out_l {parent=deform_knee_out_base_l, loc=<0 0 0>, rot=<0 0 0 1>}
 thigh_twist_01_l {parent=thigh_l, loc=<-21.6303 0 0>, rot=<0.0013 0 0 -1>}
 deform_groin_base_l {parent=thigh_l, loc=<-2.1908 8.1142 -1.7442>, rot=<0 0 0 -1>}
 deform_groin_l {parent=deform_groin_base_l, loc=<0 0 0>, rot=<0 0 0 1>}
 dyn_thighPouch_l_MSTR {parent=thigh_l, loc=<-1.5589 -0.033 -6.0147>, rot=<0.0161 -0.7121 -0.7017 -0.0181>}
 dyn_thighPouch_l_01 {parent=dyn_thighPouch_l_MSTR, loc=<0 0 0>, rot=<0 0 0 1>}
 dyn_thighPouch_l_02 {parent=dyn_thighPouch_l_01, loc=<4.1489 0.4956 -0.0225>, rot=<-0.7014 -0.0258 -0.0271 -0.7118>}
 thigh_r {parent=pelvis, loc=<-2.5693 -0.0109 7.7501>, rot=<0.0572 -0.0012 0.9982 0.0155>}
 calf_r {parent=thigh_r, loc=<39.0484 0 0>, rot=<0.0006 -0.0161 -0.0391 -0.9991>}
 foot_r {parent=calf_r, loc=<42.561 -0.1077 -0.4292>, rot=<0.0418 -0.0455 0.0343 -0.9975>}
 ball_r {parent=foot_r, loc=<6.1599 -13.8264 -0.6806>, rot=<0.0001 -0.0001 -0.7063 -0.7079>}
 calf_twist_02_r {parent=calf_r, loc=<18.8064 0.0967 -0.0397>, rot=<0.0237 0.0095 -0.0035 -0.9997>}
 dyn_pantLeg_r_01 {parent=calf_twist_02_r, loc=<0 0 0>, rot=<0 0 0 1>}
 calf_twist_01_r {parent=calf_r, loc=<35.813 0.0967 -0.0397>, rot=<0.0237 0.0095 -0.0035 -0.9997>}
 deform_knee_in_base_r {parent=calf_r, loc=<-1.298 5.2957 0.0496>, rot=<0 0 0 -1>}
 deform_knee_in_r {parent=deform_knee_in_base_r, loc=<0 0 0>, rot=<0 0 0 1>}
 deform_knee_out_base_r {parent=calf_r, loc=<0 -4 0>, rot=<0 0 0 -1>}
 deform_knee_out_r {parent=deform_knee_out_base_r, loc=<0 0 0>, rot=<0 0 0 1>}
 thigh_twist_01_r {parent=thigh_r, loc=<21.6303 0 0>, rot=<0.0013 0 0 -1>}
 deform_groin_base_r {parent=thigh_r, loc=<2.1908 -8.1142 1.7442>, rot=<0 0 0 -1>}
 deform_groin_r {parent=deform_groin_base_r, loc=<0 0 0>, rot=<0 0 0 1>}
 deform_glute_base_l {parent=pelvis, loc=<-8.8064 -8.8222 -8.4287>, rot=<0.0275 0.0501 -0.4472 -0.8926>}
 deform_glute_l {parent=deform_glute_base_l, loc=<0 0 0>, rot=<0 0 0 1>}
 deform_glute_base_r {parent=pelvis, loc=<-8.8062 -8.8222 8.4287>, rot=<0.0501 -0.0275 0.8926 -0.4472>}
 deform_glute_r {parent=deform_glute_base_r, loc=<0 0 0>, rot=<0 0 0 1>}
 ik_hand_root {parent=root, loc=<0 0 0>, rot=<0 0 0 1>}
 ik_hand_gun {parent=ik_hand_root, loc=<-51.9727 4.0883 102.6935>, rot=<0.6084 0.1793 0.4613 -0.6205>}
 ik_hand_r {parent=ik_hand_gun, loc=<0 0 0>, rot=<0 0 0 1>}
 ik_hand_l {parent=ik_hand_gun, loc=<53.0285 82.1781 35.2051>, rot=<-0.5102 -0.7906 -0.3387 0>}
 snakeAim_MSTR {controlType=EULER_TRANSFORM, parent=root, type=Control, loc=<0 0 0>, rot=<0 0 0 1>}
 snakeAim_ctrl {controlType=EULER_TRANSFORM, parent=snakeAim_MSTR, type=Control, loc=<0 0 0>, rot=<0 0 0 1>}
]
variables [
 isCrouchMoving {type=double, default=0}
 isFrontend {type=bool, default=False}
 isParachuteOpen {type=bool, default=False}
 isPlayerMoving {type=double, default=0}
 isTargeting {type=bool, default=False}
]
