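# Physics asset for the snake on the male medium SerpentCoil player body cosmetic, covering its dynamic head, tail, chest and shoulder bones.
# Capsules also cover the upper spine, neck, head, clavicles and right upper arm; these joints and the snake roots, mostly tied to spine_05, have free angular motion.
# Snake head and tail chains use narrow swing limits and zero twist, though a second, reversed tail_01/tail_02 constraint is left free.
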PhysicsAsset /BRCosmetics/Characters/Player/Male/Medium/Bodies/M_MED_SerpentCoil/Meshes/M_MED_SerpentCoil_Snake_Physics
bodies {
 clavicle_l [{shapeType=Capsule, length=23.5, radius=8.5, shapeName=clavicle_l_capsule, center=<6.0863 0 -0.973>, rotation=<rot 90 0 0>}]
 clavicle_r [{shapeType=Capsule, length=6, radius=9, shapeName=clavicle_l_capsule, center=<-6.0863 0 0>, rotation=<rot 90 180 0>}]
 dyn_snakeChest_01 [{shapeType=Capsule, length=3, radius=3, shapeName=dyn_snakeChest_01_capsule}]
 dyn_snakeHead_01 [{shapeType=Capsule, length=9.5, radius=3.5, shapeName=dyn_snakeHead_01_capsule, center=<14.5386 0 0>, rotation=<rot 90 0 0>}]
 dyn_snakeHead_02 [{shapeType=Capsule, length=6, radius=3, center=<5.1085 0 0>, rotation=<rot 90 0 0>}]
 dyn_snakeHead_03 [{shapeType=Capsule, length=6, radius=3, center=<5.1085 0 0>, rotation=<rot 90 0 0>}]
 dyn_snakeHead_04 [{shapeType=Capsule, length=5.5, radius=3.5, center=<5.1085 0 0>, rotation=<rot 90 0 0>}]
 dyn_snakeHead_MSTR [{shapeType=Capsule, length=0.505, radius=0.505, shapeName=dyn_snakeHead_MSTR_capsule, rotation=<rot 0 -90 0>}]
 dyn_snakeShoulder_l_01 [{shapeType=Capsule, length=6.5, radius=3, shapeName=dyn_snakeShoulder_l_01_capsule, rotation=<rot 0 0 -94.4>}]
 dyn_snakeShoulder_r_01 [{shapeType=Capsule, length=6.5, radius=2.5, shapeName=dyn_snakeShoulder_r_01_capsule, rotation=<rot 0 0 90>}]
 dyn_snakeTail_01 [{shapeType=Capsule, length=7.5, radius=2.5, shapeName=dyn_snakeTail_01_capsule, center=<3.7004 0 0>, rotation=<rot 90 0 0>}]
 dyn_snakeTail_02 [{shapeType=Capsule, length=7.5, radius=2.5, shapeName=dyn_snakeTail_02_capsule, center=<3.7004 0 0>, rotation=<rot 90 0 0>}]
 dyn_snakeTail_03 [{shapeType=Capsule, length=10, radius=2, shapeName=dyn_snakeTail_03_capsule, center=<5.4752 0 0>, rotation=<rot 90 -128.7493 -132.2242>}]
 dyn_snakeTail_MSTR [{shapeType=Capsule, length=0.5, radius=0.5, shapeName=dyn_snakeTail_MSTR_capsule}]
 head [{shapeType=Capsule, length=3.5, radius=11.505, shapeName=head_capsule, center=<7.2042 0 0>, rotation=<rot 0 -90 0>}]
 neck_01 [{shapeType=Capsule, length=0.5, radius=5.5, shapeName=neck_01_capsule, center=<1.9525 0 0>, rotation=<rot 90 0 0>}]
 neck_02 [{shapeType=Capsule, length=0.5, radius=5.5, shapeName=neck_02_capsule, center=<1.9525 0 0>, rotation=<rot 90 0 0>}]
 spine_04 [{shapeType=Capsule, length=13, radius=13.5, shapeName=spine_04_capsule, center=<4.5487 -0.7478 0>}]
 spine_05 [{shapeType=Capsule, length=20, radius=10.5, shapeName=spine_05_capsule, center=<2.6428 -0.6591 0>}]
 upperarm_r [{
   shapeType=Capsule
   length=20.5
   radius=8.5
   shapeName=upperarm_r_capsule
   center=<-6.5472 1.093 -0.6589>
   rotation=<rot 73.4294 -179.9997 -179.9997>
  }]
}
constraints [
 {
  bone1Name=spine_05
  bone2Name=spine_04
  swing1LimitDegrees=45
  swing1Motion=Free
  swing2LimitDegrees=45
  swing2Motion=Free
  twistLimitDegrees=45
  twistMotion=Free
 }
 {
  bone1Name=clavicle_l
  bone2Name=spine_05
  swing1LimitDegrees=45
  swing1Motion=Free
  swing2LimitDegrees=45
  swing2Motion=Free
  twistLimitDegrees=45
  twistMotion=Free
 }
 {
  bone1Name=dyn_snakeChest_01
  bone2Name=spine_05
  swing1LimitDegrees=45
  swing1Motion=Free
  swing2LimitDegrees=45
  swing2Motion=Free
  twistLimitDegrees=45
  twistMotion=Free
 }
 {
  bone1Name=dyn_snakeShoulder_l_01
  bone2Name=spine_05
  swing1LimitDegrees=45
  swing1Motion=Free
  swing2LimitDegrees=45
  swing2Motion=Free
  twistLimitDegrees=45
  twistMotion=Free
 }
 {
  bone1Name=dyn_snakeShoulder_r_01
  bone2Name=spine_05
  swing1LimitDegrees=45
  swing1Motion=Free
  swing2LimitDegrees=45
  swing2Motion=Free
  twistLimitDegrees=45
  twistMotion=Free
 }
 {
  bone1Name=dyn_snakeTail_MSTR
  bone2Name=spine_05
  swing1LimitDegrees=45
  swing1Motion=Free
  swing2LimitDegrees=45
  swing2Motion=Free
  twistLimitDegrees=45
  twistMotion=Free
 }
 {bone1Name=dyn_snakeTail_02, bone2Name=dyn_snakeTail_01, swing1LimitDegrees=20, swing2LimitDegrees=20, twistLimitDegrees=0}
 {bone1Name=dyn_snakeTail_01, bone2Name=dyn_snakeTail_MSTR, swing1LimitDegrees=5, swing2LimitDegrees=30, twistLimitDegrees=0}
 {bone1Name=dyn_snakeTail_03, bone2Name=dyn_snakeTail_02, swing1LimitDegrees=10, swing2LimitDegrees=20, twistLimitDegrees=0}
 {
  bone1Name=clavicle_r
  bone2Name=spine_05
  swing1LimitDegrees=45
  swing1Motion=Free
  swing2LimitDegrees=45
  swing2Motion=Free
  twistLimitDegrees=45
  twistMotion=Free
 }
 {
  bone1Name=dyn_snakeTail_01
  bone2Name=dyn_snakeTail_02
  swing1LimitDegrees=45
  swing1Motion=Free
  swing2LimitDegrees=45
  swing2Motion=Free
  twistLimitDegrees=45
  twistMotion=Free
 }
 {
  bone1Name=head
  bone2Name=neck_02
  swing1LimitDegrees=45
  swing1Motion=Free
  swing2LimitDegrees=45
  swing2Motion=Free
  twistLimitDegrees=45
  twistMotion=Free
 }
 {
  bone1Name=neck_01
  bone2Name=spine_05
  swing1LimitDegrees=45
  swing1Motion=Free
  swing2LimitDegrees=45
  swing2Motion=Free
  twistLimitDegrees=45
  twistMotion=Free
 }
 {
  bone1Name=neck_02
  bone2Name=neck_01
  swing1LimitDegrees=45
  swing1Motion=Free
  swing2LimitDegrees=45
  swing2Motion=Free
  twistLimitDegrees=45
  twistMotion=Free
 }
 {bone1Name=dyn_snakeHead_04, bone2Name=dyn_snakeHead_03, swing1LimitDegrees=0, swing2LimitDegrees=15, twistLimitDegrees=0}
 {bone1Name=dyn_snakeHead_03, bone2Name=dyn_snakeHead_02, swing1LimitDegrees=0, swing2LimitDegrees=15, twistLimitDegrees=0}
 {bone1Name=dyn_snakeHead_02, bone2Name=dyn_snakeHead_01, swing1LimitDegrees=0, swing2LimitDegrees=15, twistLimitDegrees=0}
 {bone1Name=dyn_snakeHead_01, bone2Name=dyn_snakeHead_MSTR, swing1LimitDegrees=12, swing2LimitDegrees=35, twistLimitDegrees=0}
 {
  bone1Name=dyn_snakeHead_MSTR
  bone2Name=spine_05
  swing1LimitDegrees=45
  swing1Motion=Free
  swing2LimitDegrees=45
  swing2Motion=Free
  twistLimitDegrees=45
  twistMotion=Free
 }
 {
  bone1Name=upperarm_r
  bone2Name=clavicle_r
  swing1LimitDegrees=45
  swing1Motion=Free
  swing2LimitDegrees=45
  swing2Motion=Free
  twistLimitDegrees=45
  twistMotion=Free
 }
]
