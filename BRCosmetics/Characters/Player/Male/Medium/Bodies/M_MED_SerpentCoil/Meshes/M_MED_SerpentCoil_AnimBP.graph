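# Post-process style AnimBlueprint for the SerpentCoil male cosmetic body, animating its snake, cloak, vest and pant legs over the parent mesh's copied pose.
# Cached stages run a deform Control Rig, bone offsets, cloak physics switched by backbling, then snake, vest and pant dynamics.
# A snake Control Rig driven by movement, targeting, parachute and frontend state precedes final chain dynamics.
# The EventGraph eases movement values, sets physics alpha and updates material WPO.
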AnimBlueprint /BRCosmetics/Characters/Player/Male/Medium/Bodies/M_MED_SerpentCoil/Meshes/M_MED_SerpentCoil_AnimBP
target_skeleton=&/M_MED_SerpentCoil_Skeleton
graphs [
 AnimGraph {nodes [
   n1 = Control Rig!(
    Source=Use cached pose 'dynamicsCache'()
    Alpha
    isPlayerMoving:double=Get isPlayerMoving() -> (isPlayerMoving:double)
    isTargeting:bool=Get bIsTargeting()
    isParachuteOpen:bool=Get bIsParachuteOpened()
    isCrouchMoving:double=Get isCrouchMoving() -> (isCrouchMoving:double)
    isFrontend:bool=Get bIsInFrontEnd()
   ) -> (Pose) {
    alphaBoolBlend {blendOption=Linear}
    controlRigAssetReference {blueprintRigClass=&/M_MED_SerpentCoil_Snake_CtrlRig.M_MED_SerpentCoil_Snake_CtrlRig_C}
   }
   n2 = Anim Dynamics - Bone: dyn_snakeChain_01(ComponentPose=Local To Component(LocalPose=@n1), Alpha=1, ExternalForce="0,0,0") {
    angularDampingOverride=1
    angularSpringConstant=20000
    bOverrideAngularDamping=true
    bUseGravityOverride=true
    gravityScale=0.35
    simulationSpace=World
    alphaBoolBlend {blendOption=Linear}
    boundBone {boneName=dyn_snakeChain_01}
    componentAppliedLinearAccClamp=<6000 6000 6000>
    componentLinearAccScale=<1 1 1>
    componentLinearVelScale=<1 1 1>
    gravityOverride {z=-980}
    physicsBodyDefinitions [{
      boundBone {boneName=dyn_snakeChain_01}
      boxExtents=<20 30 30>
      constraintSetup {angularLimitsMax=<30 30 30>, angularLimitsMin=<-30 -30 -30>, angularTarget=<1 0 0>}
      localJointOffset=<12 0 0>
     }]
    relativeSpaceBone {boneName=root}
   }
   n3 = Anim Dynamics - Bone: dyn_snakeString_01(ComponentPose=@n2, Alpha=1, ExternalForce="0,0,0") {
    angularDampingOverride=1
    angularSpringConstant=20000
    bOverrideAngularDamping=true
    bUseGravityOverride=true
    gravityScale=0.35
    simulationSpace=World
    alphaBoolBlend {blendOption=Linear}
    boundBone {boneName=dyn_snakeString_01}
    componentAppliedLinearAccClamp=<6000 6000 6000>
    componentLinearAccScale=<1 1 1>
    componentLinearVelScale=<1 1 1>
    gravityOverride {z=-980}
    physicsBodyDefinitions [{
      boundBone {boneName=dyn_snakeString_01}
      boxExtents=<20 30 30>
      constraintSetup {angularLimitsMax=<20 20 40>, angularLimitsMin=<-20 -20 -40>, angularTarget=<1 0 0>}
      localJointOffset=<10 0 0>
     }]
    relativeSpaceBone {boneName=root}
   }
   n4 = "Transform (Modify) Bone - Bone: dyn_backCloak_l_MSTR"!(ComponentPose=@n3, bAlphaBoolEnabled:bool=True, Translation="0,0,0", Rotation="0,-20,0", Scale="1,1,1") -> (Pose) {
    alphaInputType=Bool
    rotationMode=BMM_Additive
    rotationSpace=BCS_BoneSpace
    alphaBoolBlend {blendInTime=0.5, blendOption=Linear, blendOutTime=0.5}
    boneToModify {boneName=dyn_backCloak_l_MSTR}
    rotation {yaw=-20}
   }
   n5 = "Transform (Modify) Bone - Bone: dyn_backCloak_r_MSTR"!(ComponentPose=@n4, bAlphaBoolEnabled:bool=True, Translation="0,0,0", Rotation="0,-20,0", Scale="1,1,1") -> (Pose) {
    alphaInputType=Bool
    rotationMode=BMM_Additive
    rotationSpace=BCS_BoneSpace
    alphaBoolBlend {blendInTime=0.5, blendOption=Linear, blendOutTime=0.5}
    boneToModify {boneName=dyn_backCloak_r_MSTR}
    rotation {yaw=-20}
   }
   Output Pose(Result=Component To Local(ComponentPose=@n5)) {layerGroup=None}
   n6 = Control Rig!(Source=Copy Pose From Mesh() {bUseAttachedParent=true}, Alpha) -> (Pose) {
    alphaBoolBlend {blendOption=Linear}
    controlRigAssetReference {blueprintRigClass=&/M_MED_DeformRig_SerpentCoil_ControlRig.M_MED_DeformRig_SerpentCoil_ControlRig_C}
   }
   Save cached pose 'CRCache'(Pose=@n6) {cacheName=CRCache}
   n7 = RigidBody!(
    ComponentPose=Local To Component(LocalPose=Use cached pose 'necklaceCache'())
    Alpha=Get RBANAlphaVals() -> (RBANAlphaVals:double)
    OverrideWorldGravity:Vector="0,0,-980"
    ComponentLinearAccScale:Vector="0.25,0.6,0.6"
    ComponentLinearVelScale:Vector="0.25,0.6,0.6"
   ) -> (Pose) {
    bClampLinearTranslationLimitToRefPose=true
    bOverrideWorldGravity=true
    overridePhysicsAsset=&/M_MED_SerpentCoil_Snake_Physics
    simulationSpace=BaseBoneSpace
    alphaBoolBlend {blendOption=Linear}
    baseBoneRef {boneName=spine_05}
    componentAppliedLinearAccClamp=<6000 6000 6000>
    componentLinearAccScale=<0.25 0.6 0.6>
    componentLinearVelScale=<0.25 0.6 0.6>
    overrideWorldGravity {z=-980}
   }
   n8 = "Transform (Modify) Bone - Bone: dyn_snakeTail_03"(ComponentPose=@n7, Alpha=0.5, Translation="0,0,0", Rotation="30,-10,-20", Scale="1,1,1") {
    rotationMode=BMM_Additive
    rotationSpace=BCS_BoneSpace
    alphaBoolBlend {blendOption=Linear}
    boneToModify {boneName=dyn_snakeTail_03}
    rotation=<rot 30 -10 -20>
   }
   n9 = Component To Local(ComponentPose=@n8)
   n10 = RigidBody(
    ComponentPose=Local To Component(LocalPose=Use cached pose 'necklaceCache'())
    Alpha=Get RBANAlphaVals()
    OverrideWorldGravity="0,0,-980"
    ComponentLinearAccScale="0.25,0.6,0.6"
    ComponentLinearVelScale="0.25,0.6,0.6"
   ) {
    bClampLinearTranslationLimitToRefPose=true
    bOverrideWorldGravity=true
    overridePhysicsAsset=&M_MED_SerpentCoil_Snake_Physics
    simulationSpace=BaseBoneSpace
    alphaBoolBlend {blendOption=Linear}
    baseBoneRef {boneName=spine_05}
    componentAppliedLinearAccClamp=<6000 6000 6000>
    componentLinearAccScale=<0.25 0.6 0.6>
    componentLinearVelScale=<0.25 0.6 0.6>
    overrideWorldGravity {z=-980}
   }
   n11 = Component To Local(ComponentPose=@n10)
   n12 = Get bIsInFrontEnd()
   n13 = RigidBody!(
    ComponentPose=Local To Component(LocalPose=Blend Poses by bool(BlendPose_0=@n9, BlendPose_1=@n11, bActiveValue=@n12) {blendType=Linear})
    Alpha
    OverrideWorldGravity="0,0,-980"
    ExternalForce:Vector
   ) -> (Pose) {
    bClampLinearTranslationLimitToRefPose=true
    bOverrideWorldGravity=true
    overridePhysicsAsset=&/M_MED_SerpentCoil_LowerVest_Physics
    simulationSpace=BaseBoneSpace
    alphaBoolBlend {blendOption=Linear}
    baseBoneRef {boneName=spine_01}
    componentAppliedLinearAccClamp=<6000 6000 6000>
    componentLinearAccScale=<0.5 0.2 0.2>
    componentLinearVelScale=<0.5 0.2 0.2>
    overrideWorldGravity {z=-980}
   }
   n14 = Anim Dynamics - Bone: dyn_pantLeg_l_01(ComponentPose=@n13, Alpha=1, ExternalForce="0,0,0") {
    angularDampingOverride=1
    angularSpringConstant=20000
    bOverrideAngularDamping=true
    bUseGravityOverride=true
    gravityScale=0.35
    simulationSpace=BoneRelative
    alphaBoolBlend {blendOption=Linear}
    boundBone {boneName=dyn_pantLeg_l_01}
    componentAppliedLinearAccClamp=<6000 6000 6000>
    componentLinearAccScale=<0.6 0.6 0.6>
    componentLinearVelScale=<0.6 0.6 0.6>
    gravityOverride {z=-980}
    physicsBodyDefinitions [{
      boundBone {boneName=dyn_pantLeg_l_01}
      boxExtents=<20 30 30>
      constraintSetup {angularLimitsMax=<5 5 2>, angularLimitsMin=<-5 -5 -2>, angularTarget=<1 0 0>}
      localJointOffset=<7 0 0>
     }]
    relativeSpaceBone {boneName=calf_l}
   }
   n15 = Anim Dynamics - Bone: dyn_pantLeg_r_01(ComponentPose=@n14, Alpha=1, ExternalForce="0,0,0") {
    angularDampingOverride=1
    angularSpringConstant=20000
    bOverrideAngularDamping=true
    bUseGravityOverride=true
    gravityScale=0.35
    simulationSpace=BoneRelative
    alphaBoolBlend {blendOption=Linear}
    boundBone {boneName=dyn_pantLeg_r_01}
    componentAppliedLinearAccClamp=<6000 6000 6000>
    componentLinearAccScale=<0.6 0.6 0.6>
    componentLinearVelScale=<0.6 0.6 0.6>
    gravityOverride {z=-980}
    physicsBodyDefinitions [{
      boundBone {boneName=dyn_pantLeg_r_01}
      boxExtents=<20 30 30>
      constraintSetup {angularLimitsMax=<5 5 2>, angularLimitsMin=<-5 -5 -2>, angularTarget=<1 0 0>}
      localJointOffset=<7 0 0>
     }]
    relativeSpaceBone {boneName=calf_r}
   }
   n16 = Anim Dynamics - Bone: dyn_sleeve_l_01(ComponentPose=@n15, Alpha=0, ExternalForce="0,0,0") {
    angularDampingOverride=1
    angularSpringConstant=20000
    bOverrideAngularDamping=true
    bUseGravityOverride=true
    gravityScale=0.35
    simulationSpace=BoneRelative
    alphaBoolBlend {blendOption=Linear}
    boundBone {boneName=dyn_sleeve_l_01}
    componentAppliedLinearAccClamp=<6000 6000 6000>
    componentLinearAccScale=<0.35 0.35 0.35>
    componentLinearVelScale=<0.35 0.35 0.35>
    gravityOverride {z=-980}
    physicsBodyDefinitions [{
      boundBone {boneName=dyn_sleeve_l_01}
      boxExtents=<15 30 30>
      constraintSetup {angularLimitsMax=<2 2 2>, angularLimitsMin=<-2 -2 -2>, angularTarget=<1 0 0>}
      localJointOffset=<6 0 0>
     }]
    relativeSpaceBone {boneName=upperarm_l}
   }
   n17 = Anim Dynamics - Bone: dyn_sleeve_r_01(ComponentPose=@n16, Alpha=0, ExternalForce="0,0,0") {
    angularDampingOverride=1
    angularSpringConstant=20000
    bOverrideAngularDamping=true
    bUseGravityOverride=true
    gravityScale=0.35
    simulationSpace=BoneRelative
    alphaBoolBlend {blendOption=Linear}
    boundBone {boneName=dyn_sleeve_r_01}
    componentAppliedLinearAccClamp=<6000 6000 6000>
    componentLinearAccScale=<0.2 0.2 0.2>
    componentLinearVelScale=<0.2 0.2 0.2>
    gravityOverride {z=-980}
    physicsBodyDefinitions [{
      boundBone {boneName=dyn_sleeve_r_01}
      boxExtents=<15 30 30>
      constraintSetup {angularLimitsMax=<2 2 2>, angularLimitsMin=<-2 -2 -2>, angularTarget=<1 0 0>}
      localJointOffset=<6 0 0>
     }]
    relativeSpaceBone {boneName=upperarm_r}
   }
   Save cached pose 'dynamicsCache'(Pose=Component To Local(ComponentPose=@n17)) {cacheName=dynamicsCache}
   n18 = "Constraint - dyn_zipperVest_MSTR (PPPPP)"!(
    ComponentPose=Local To Component(LocalPose=Use cached pose 'CRCache'())
    Alpha=1
    ConstraintWeights_0=0.5723
    ConstraintWeights_1:float=0.1713
    ConstraintWeights_2:float=0.1446
    ConstraintWeights_3:float=0.1031
    ConstraintWeights_4:float=0.0087
   ) -> (Pose) {
    alphaBoolBlend {blendOption=Linear}
    boneToModify {boneName=dyn_zipperVest_MSTR}
    constraintSetup [
     {targetBone {boneName=pelvis}}
     {targetBone {boneName=spine_02}}
     {targetBone {boneName=spine_01}}
     {targetBone {boneName=spine_03}}
     {targetBone {boneName=spine_04}}
    ]
    constraintWeights [1,1,1,1,1]
   }
   n19 = Anim Dynamics - Bone: dyn_zipperVest_01(ComponentPose=@n18, Alpha=1, ExternalForce="0,0,0") {
    angularDampingOverride=1
    angularSpringConstant=20000
    bAngularSpring=true
    bOverrideAngularDamping=true
    bUseGravityOverride=true
    gravityScale=0.35
    simulationSpace=RootRelative
    alphaBoolBlend {blendOption=Linear}
    boundBone {boneName=dyn_zipperVest_01}
    componentLinearAccScale=<1 1 1>
    componentLinearVelScale=<1 1 1>
    gravityOverride {z=-980}
    physicsBodyDefinitions [{
      boundBone {boneName=dyn_zipperVest_01}
      boxExtents=<20 30 30>
      constraintSetup {angularLimitsMax=<5 22 30>, angularLimitsMin=<-5 -22 -5>, angularTarget=<1 0 0>}
      localJointOffset=<8 0 0>
     }]
   }
   n20 = "Transform (Modify) Bone - Bone: dyn_snake_MSTR"!(ComponentPose=@n19, bAlphaBoolEnabled:bool=True, Translation="2.5,-3,0", Rotation="0,0,0", Scale="1,1,1") -> (Pose) {
    alphaInputType=Bool
    rotationMode=BMM_Additive
    rotationSpace=BCS_BoneSpace
    translationMode=BMM_Additive
    translationSpace=BCS_BoneSpace
    alphaBoolBlend {blendInTime=0.5, blendOption=Linear, blendOutTime=0.2}
    boneToModify {boneName=dyn_snake_MSTR}
    translation=<2.5 -3>
   }
   n21 = "Transform (Modify) Bone - Bone: dyn_snakeTail_01"!(ComponentPose=@n20, bAlphaBoolEnabled:bool=True, Translation="0,0,4", Rotation="0,0,0", Scale="1,1,1") -> (Pose) {
    alphaInputType=Bool
    rotationMode=BMM_Additive
    rotationSpace=BCS_BoneSpace
    translationMode=BMM_Additive
    translationSpace=BCS_BoneSpace
    alphaBoolBlend {blendInTime=0.5, blendOption=Linear, blendOutTime=0.2}
    boneToModify {boneName=dyn_snakeTail_01}
    translation {z=4}
   }
   n22 = "Transform (Modify) Bone - Bone: dyn_snakeHead_01"!(ComponentPose=@n21, bAlphaBoolEnabled:bool=True, Translation="0,0,3", Rotation="0,0,0", Scale="1,1,1") -> (Pose) {
    alphaInputType=Bool
    rotationMode=BMM_Additive
    rotationSpace=BCS_BoneSpace
    translationMode=BMM_Additive
    translationSpace=BCS_BoneSpace
    alphaBoolBlend {blendInTime=0.4, blendOption=Linear, blendOutTime=0.2}
    boneToModify {boneName=dyn_snakeHead_01}
    translation {z=3}
   }
   n23 = "Transform (Modify) Bone - Bone: dyn_snakeShoulder_l_01"!(ComponentPose=@n22, bAlphaBoolEnabled:bool=True, Translation="-5,0,5", Rotation="0,0,0", Scale="1,1,1") -> (Pose) {
    alphaInputType=Bool
    rotationMode=BMM_Additive
    rotationSpace=BCS_BoneSpace
    translationMode=BMM_Additive
    translationSpace=BCS_BoneSpace
    alphaBoolBlend {blendInTime=0.8, blendOption=Linear, blendOutTime=0.6}
    boneToModify {boneName=dyn_snakeShoulder_l_01}
    translation {x=-5, z=5}
   }
   n24 = "Transform (Modify) Bone - Bone: dyn_snakeShoulder_r_01"!(ComponentPose=@n23, bAlphaBoolEnabled:bool=True, Translation="-6,0,-5", Rotation="0,0,0", Scale="1,1,1") -> (Pose) {
    alphaInputType=Bool
    rotationMode=BMM_Additive
    rotationSpace=BCS_BoneSpace
    translationMode=BMM_Additive
    translationSpace=BCS_BoneSpace
    alphaBoolBlend {blendInTime=0.8, blendOption=Linear, blendOutTime=0.4}
    boneToModify {boneName=dyn_snakeShoulder_r_01}
    translation {x=-6, z=-5}
   }
   n25 = "Transform (Modify) Bone - Bone: dyn_snakeChest_01"!(ComponentPose=@n24, bAlphaBoolEnabled:bool=True, Translation="-6,6,0", Rotation="0,0,0", Scale="1,1,1") -> (Pose) {
    alphaInputType=Bool
    rotationMode=BMM_Additive
    rotationSpace=BCS_BoneSpace
    translationMode=BMM_Additive
    translationSpace=BCS_BoneSpace
    alphaBoolBlend {blendInTime=0.3, blendOption=Linear, blendOutTime=0.5}
    boneToModify {boneName=dyn_snakeChest_01}
    translation=<-6 6>
   }
   n26 = "Transform (Modify) Bone - Bone: dyn_snakeHat_01"!(ComponentPose=@n25, bAlphaBoolEnabled:bool=True, Translation="-6,0,0", Rotation="0,0,0", Scale="1,1,1") -> (Pose) {
    alphaInputType=Bool
    rotationMode=BMM_Additive
    rotationSpace=BCS_BoneSpace
    translationMode=BMM_Additive
    translationSpace=BCS_BoneSpace
    alphaBoolBlend {blendInTime=1.5, blendOption=Linear, blendOutTime=0.3}
    boneToModify {boneName=dyn_snakeHat_01}
    translation {x=-6}
   }
   n27 = "Transform (Modify) Bone - Bone: dyn_snakeTail_MSTR"!(ComponentPose=@n26, bAlphaBoolEnabled:bool=Get bIsSkydiveFloating(), Translation="0,-2,0", Rotation="0,0,20", Scale="1,1,1") -> (Pose) {
    alphaInputType=Bool
    rotationMode=BMM_Additive
    rotationSpace=BCS_BoneSpace
    translationMode=BMM_Additive
    translationSpace=BCS_BoneSpace
    alphaBoolBlend {blendOption=Linear}
    boneToModify {boneName=dyn_snakeTail_MSTR}
    rotation {roll=20}
    translation {y=-2}
   }
   Save cached pose 'preSimCache'(Pose=Component To Local(ComponentPose=@n27)) {cacheName=preSimCache}
   n28 = RigidBody!(
    ComponentPose=Local To Component(LocalPose=Use cached pose 'preSimCache'())
    Alpha
    OverrideWorldGravity="0,0,-980"
    ExternalForce
    ComponentLinearAccScale:Vector="0.35,0.5,0.5"
    ComponentLinearVelScale:Vector="0.35,0.5,0.5"
   ) -> (Pose) {
    bClampLinearTranslationLimitToRefPose=true
    bOverrideWorldGravity=true
    overridePhysicsAsset=&/M_MED_SerpentCoil_Backbling_cloak_Physics
    simulationSpace=BaseBoneSpace
    alphaBoolBlend {blendOption=Linear}
    baseBoneRef {boneName=spine_05}
    componentAppliedLinearAccClamp=<6000 6000 6000>
    componentLinearAccScale=<0.35 0.5 0.5>
    componentLinearVelScale=<0.35 0.5 0.5>
    overrideWorldGravity {z=-980}
   }
   n29 = "Constraint - dyn_backCloak_02 (PPPPPPPP)"!(
    ComponentPose=@n28
    Alpha=1
    ConstraintWeights_0=0.4039
    ConstraintWeights_1:float=0.2233
    ConstraintWeights_2:float=0.1627
    ConstraintWeights_3:float=0.0937
    ConstraintWeights_4:float=0.0521
    ConstraintWeights_5:float=0.0275
    ConstraintWeights_6:float=0.019
    ConstraintWeights_7:float=0.0178
   ) -> (Pose) {
    alphaBoolBlend {blendOption=Linear}
    boneToModify {boneName=dyn_backCloak_02}
    constraintSetup [
     {targetBone {boneName=spine_03}}
     {targetBone {boneName=spine_02}}
     {targetBone {boneName=spine_04}}
     {targetBone {boneName=spine_01}}
     {targetBone {boneName=pelvis}}
     {targetBone {boneName=clavicle_r}}
     {targetBone {boneName=spine_05}}
     {targetBone {boneName=clavicle_l}}
    ]
    constraintWeights [1,1,1,1,1,1,1,1]
   }
   n30 = RigidBody(
    ComponentPose=Local To Component(LocalPose=Use cached pose 'preSimCache'())
    OverrideWorldGravity="0,0,-980"
    ComponentLinearAccScale="0.3,0.6,0.5"
    ComponentLinearVelScale="0.3,0.6,0.5"
   ) {
    bClampLinearTranslationLimitToRefPose=true
    bOverrideWorldGravity=true
    overridePhysicsAsset=&/M_MED_SerpentCoil_Necklace_Cloak_Physics
    simulationSpace=BaseBoneSpace
    alphaBoolBlend {blendOption=Linear}
    baseBoneRef {boneName=spine_05}
    componentAppliedLinearAccClamp=<6000 6000 6000>
    componentLinearAccScale=<0.3 0.6 0.5>
    componentLinearVelScale=<0.3 0.6 0.5>
    overrideWorldGravity {z=-980}
   }
   Save cached pose 'necklaceCache'(Pose=Blend Poses by bool(
     BlendPose_0=Component To Local(ComponentPose=@n29)
     BlendPose_1=Component To Local(ComponentPose=@n30)
     bActiveValue=Get isWearingBackbling()
    ) {blendType=Linear}) {cacheName=necklaceCache}
  ]}
 EventGraph {nodes [
   n1 = Event BlueprintUpdateAnimation()
   > Set Delta Time X(Delta Time X=@n1.DeltaTimeX)
   > n2 = Sequence!(execute) -> (then_0, then_1, then_2:Exec, then_3:Exec, then_4:Exec, then_5:Exec, then_6:Exec, then_7:Exec, then_8:Exec, then_9:Exec)
   n3 = Sequence!(execute=@n2.then_7) -> (then_0, then_1, then_2)
   > n4 = Branch(Condition=Get bIsSkydiveFloating())
   > Set RBANAlphaVals(RBANAlphaVals:double=0) -> (Output_Get:double)
   Set RBANAlphaVals(execute=@n4.else, RBANAlphaVals=0.4)
   n5 = Branch(execute=@n3.then_1, Condition=AND Boolean(A=Get bIsCrouching(), B=Get isPlayerMovingBool() -> (isPlayerMovingBool:bool)))
   > Set isCrouchMoving(isCrouchMoving:double=FInterpTo(Current=Get isCrouchMoving(), Target=1, DeltaTime=Get Delta Time X(), InterpSpeed=3)) -> (Output_Get:double)
   Set isCrouchMoving(execute=@n5.else, isCrouchMoving=FInterpTo(Current=Get isCrouchMoving(), Target=0, DeltaTime=Get Delta Time X(), InterpSpeed=5))
   n6 = Branch(execute=@n3.then_2, Condition=Get bIsInFrontEnd())
   > Set FEAccSclVals(FEAccSclVals:Vector="0.01,0.01,0.01") -> (Output_Get:Vector)
   Set FEAccSclVals(execute=@n6.else, FEAccSclVals="0.3,0.5,0.5")
   n7 = Set CurrentVar(
    execute=@n2.then_8
    CurrentVar:double=FInterpTo(
     Current=Get CurrentVar() -> (CurrentVar:double)
     Target=Get myFloatOverTime() -> (myFloatOverTime:double)
     DeltaTime=Get Delta Time X()
     InterpSpeed=0.1
    )
   ) -> (Output_Get:double)
   Set timeBool(execute=@n2.then_9, timeBool:bool=false) -> (Output_Get:bool)
   > RetriggerableDelay(Duration=0.1)
   > Set timeBool(timeBool=true)
   n8 = Branch(execute=@n2.then_6, Condition=Get bIsParachuteOpened())
   > Set accVelScalVals(accVelScalVals:Vector="0, 0, 0") -> (Output_Get:Vector)
   Set accVelScalVals(execute=@n8.else, accVelScalVals="0.25,0.6,0.6")
   n9 = Sequence(execute=@n2.then_3)
   > n10 = Set headRotateY(headRotateY:float=GetAlphaFromReferenceTransformDelta(
     SkeletalMeshComponent=Get BodyMesh() -> (BodyMesh:Skeletal Mesh Component Object Reference)
     BoneName=head
     BaseName=spine_05
     DeltaComponentType=EulerY
     InputRange=(X=0,Y=1)
     TwistAxis="1,0,0"
    ).RawOutput) -> (Output_Get:float)
   n11 = Set Raw Output(
    execute=@n9.then_1
    Raw Output:float=GetAlphaFromReferenceTransformDelta(SkeletalMeshComponent=Get BodyMesh(), BoneName=head, BaseName=spine_05, DeltaComponentType=EulerX, InputRange=(X=0,Y=1), TwistAxis="1,0,0").RawOutput
   ) -> (Output_Get:float)
   n12 = float * float(A=@n11, B=-1) {is=title K2Node_PromotableOperator}
   n13 = Set headRotateZ(
    execute=@n9.then_2
    headRotateZ:float=GetAlphaFromReferenceTransformDelta(SkeletalMeshComponent=Get BodyMesh(), BoneName=head, BaseName=spine_05, InputRange=(X=0,Y=1), TwistAxis="1,0,0").RawOutput
   ) -> (Output_Get:float)
   n14 = float * float(A=@n13, B=-1)
   Set headRotationVals(execute=@n10.then, headRotationVals:Rotator=MakeRotator(Roll=@n10, Pitch=@n14, Yaw=@n12)) -> (Output_Get:Rotator)
   > Set headRotationQuat(headRotationQuat:Quat Structure=MakeQuat!(X:float=@n10, Y:float=@n14, Z:float=@n12, W:float=1) -> (ReturnValue:Quat Structure)) -> (Output_Get:Quat Structure)
   n15 = Gravity Override State Function DEPRECATED(
    execute=@n2.then_0
    CharacterPartAnimInstance=Self-Reference()
    GameTimeInSeconds=Get Delta Time X()
    GravityOverrideParams=Get Gravity Override Params()
   )
   > Set Gravity Override(Gravity Override=@n15)
   Set BodyMesh(execute=@n2.then_2, BodyMesh=GetPartSkeletalMeshComponent())
   Set WPO Globals Function(
    execute=@n2.then_5
    Material=Get Mat Instance Dynamic()
    Skeletal Mesh=Get BodyMesh()
    "Local Noise Transform  Location"=Get Local Noise Transform Location() -> (Local Noise Transform Location:Vector)
    Local Noise Offset=Get Local Noise Offset()
   )
   > Set WPO Regional Function!(
    execute
    Material:Material Instance Dynamic Object Reference=Get Mat Instance Dynamic()
    Overlay Material:Material Instance Dynamic Object Reference
    Skeletal Mesh:Skeletal Mesh Component Object Reference=Get BodyMesh()
    Velocity Transform:Name=spine_05
    Velocity Interp Speed:double=Get Velocity Interp Speed()
    Collision Transform:Name=spine_05
    Collision Transform Axis:EAxis Enum=Y
    Invert Collision Transform Axis:bool=true
    Delta Time X:double=Get Delta Time X()
    Previous Velocity Transform Location:"Vector (by ref)"=Get R Velocity Transform Location() -> (R Velocity Transform Location:Vector)
    Previous Velocity Transform Velocity:"Vector (by ref)"=Get R Velocity Transform Velocity() -> (R Velocity Transform Velocity:Vector)
    Region:int=1
   ) -> (then)
   n16 = Sequence!(execute=@n2.then_1) -> (then_0, then_1)
   > Set Riding Speed 2D(Riding Speed 2D:float=Get PawnSpeed2D()) -> (Output_Get:float)
   n17 = Get Riding Speed 2D() -> (Riding Speed 2D:float)
   > n18 = Branch(Condition=OR Boolean(A="float > float"(A=@n17), B=Get bIsJumping()))
   > Set isPlayerMovingBool(isPlayerMovingBool:bool=true) -> (Output_Get:bool)
   Set isPlayerMovingBool(execute=@n18.else, isPlayerMovingBool=false)
   n19 = Branch(execute=@n16.then_1, Condition=Get isPlayerMovingBool())
   > Set isPlayerMoving(isPlayerMoving:double=FInterpTo(Current=Get isPlayerMoving(), Target=0, DeltaTime=Get Delta Time X(), InterpSpeed=0.2)) -> (Output_Get:double)
   Set isPlayerMoving(execute=@n19.else, isPlayerMoving=FInterpTo(Current=Get isPlayerMoving(), Target=1, DeltaTime=Get Delta Time X(), InterpSpeed=0.5))
   n20 = Get bIsSkyDiving()
   n21 = Branch(execute=@n2.then_4, Condition=AND Boolean(A=Get bIsFalling(), B=NOT Boolean(A=@n20)))
   > Set isFallingNotSkydiving(isFallingNotSkydiving:bool=true) -> (Output_Get:bool)
   Set isFallingNotSkydiving(execute=@n21.else, isFallingNotSkydiving=false)
   Event OnBeginJumping()
   > Set startedJumping(startedJumping:bool=true) -> (Output_Get:bool)
   Event OnCharacterCustomizationLoaded()
   > n22 = Sequence()
   n23 = CheckForBackbling()
   > n24 = Branch(Condition=OR Boolean!(A=@n23, B=@n23.bIsBackblingACape, C:bool=@n23.bIsBackblingLong) -> (ReturnValue))
   > Set isWearingBackbling(isWearingBackbling=true)
   Set isWearingBackbling(execute=@n24.else, isWearingBackbling=false)
   n25 = Get Dynamic Material Instance(execute=@n22.then_1, SkeletalMeshComponent=Get BodyMesh())
   > Set Mat Instance Dynamic(Mat Instance Dynamic=@n25)
   Event OnEndJumping()
   > Set startedJumping(startedJumping=false)
   MakeRotator(Roll=@n7, Pitch=@n7, Yaw=@n7)
   Set myFloatOverTime(myFloatOverTime:double="To Float (Integer)"(InInt=RandomIntegerInRange!(Min:int=-10, Max:int=10) -> (ReturnValue:int=0))) -> (Output_Get:double)
  ]}
]
variables [
 accVelScalVals {type=Vector}
 BodyMesh {type=SkeletalMeshComponent}
 CurrentVar {type=number, default=1}
 Delta Time X {type=number}
 FEAccSclVals {type=Vector}
 Gravity Override {type=Vector}
 Gravity Override Params {type=GravityOverrideParamsStruct, default {skydive_Parachuting {gravityOverride=<0 -200 -980>, windAmplitude=<0 -500 250>}}}
 headRandomRotVal {type=Rotator}
 headRotateX {type=number}
 headRotateY {type=number}
 headRotateZ {type=number}
 headRotationQuat {type=Quat, default=<0 0 0 1>}
 headRotationVals {type=Rotator}
 isCrouchMoving {type=number}
 isFallingNotSkydiving {type=boolean}
 isPlayerMoving {type=number}
 isPlayerMovingBool {type=boolean}
 isWearingBackbling {type=boolean}
 Local Noise Offset {type=Vector}
 Local Noise Transform Location {type=Vector}
 Mat Instance Dynamic {type=MaterialInstanceDynamic}
 myFloatOverTime {type=number}
 R Velocity Transform Location {type=Vector}
 R Velocity Transform Velocity {type=Vector}
 Raw Output {type=number}
 RBANAlphaVals {type=number}
 Riding Speed 2D {type=number}
 Source Player {type=FortPlayerController, default=null}
 startedJumping {type=boolean}
 timeBool {type=boolean}
 Velocity Interp Speed {type=number}
]
